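# Material function for the Lunch Box character body that layers an animated two-color emissive, by default a lightning gradient, onto incoming material attributes.
# It projects a time-panned texture triplanar from pre-skinned local position and normal, computed through vertex interpolators by default.
# The red channel drives brightness and darkens base color, while Secondary Mask picks between the two colors.
# Gradient Mask offset by Reveal, times Primary Mask, controls the blend over the original attributes.
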MaterialFunction /BRCosmetics/Characters/Player/Male/Medium/Bodies/M_MED_Lunch_Box/Materials/MF_Lunch_Box_Emissive
graph {nodes [
  n1 = FunctionInput() {inputName=Material, inputType=FunctionInput_MaterialAttributes}
  n2 = GetMaterialAttributes(None=@n1) {attributeGetTypes [#1]}
  n3 = Multiply(
   A=FunctionInput() {bUsePreviewValueAsDefault=true, inputName=UVOffset, inputType=FunctionInput_Scalar}
   B=AppendVector(A=ScalarParameter() {group=PanningEmissive, parameterName=UVOffsetU}, B=ScalarParameter() {group=PanningEmissive, parameterName=UVOffsetV})
  )
  n4 = ScalarParameter() {defaultValue=1, group=PanningEmissive, parameterName=Emissive Tiling}
  n5 = PreSkinnedPosition()
  n6 = Divide(
   A=FunctionInput(Preview=VertexInterpolator(VS=@n5).PS) {bUsePreviewValueAsDefault=true, inputName=Local Position}
   B=ScalarParameter() {defaultValue=40, group=PanningEmissive, parameterName=Triplanner_ConstantScaleDivisor}
  )
  n7 = Frac(None=Multiply(A=Time(), B=VectorParameter() {group=PanningEmissive, parameterName=TriPlanner Speed, defaultValue=<rgba 0 0.2 0.2 1>}.RGB))
  n8 = Add(A=Multiply(A=@n4, B=@n6), B=@n7)
  n9 = TextureObjectParameter() {
   parameterName=Panning Texture
   samplerType=SAMPLERTYPE_LinearColor
   texture=/Game/Characters/Player/Male/Medium/Bodies/M_MED_Lunch_Box/Textures/Lightning_Grad_M
  }
  n10 = TextureSample(UVs=Add(A=@n3, B=ComponentMask(None=@n8) {b=true, g=true}), Tex=@n9) {texture=/Game/Effects/Fort_Effects/Textures/Grad/T_FX_LinearSweep_Gradient}
  n11 = Add(A=Multiply(A=@n4, B=@n6), B=@n7)
  n12 = TextureSample(UVs=Add(A=@n3, B=ComponentMask(None=@n11) {b=true, r=true}), Tex=@n9) {texture=&T_FX_LinearSweep_Gradient}
  n13 = FunctionInput(Preview=VertexInterpolator(VS=PreSkinnedNormal()).PS) {bUsePreviewValueAsDefault=true, inputName=Local Normal}
  n14 = Abs(None=ComponentMask(None=@n13) {g=true})
  n15 = ScalarParameter() {defaultValue=1, group=PanningEmissive, parameterName=TriPlanner_AxisBlendStrength}
  n16 = Saturate(None=ComponentMask(None=MaterialFunctionCall(In=@n14, Contrast=@n15) {materialFunction=&CheapContrast}.Result) {g=true})
  n17 = Multiply(A=@n4, B=@n6)
  n18 = Add(A=@n3, B=ComponentMask(None=Add(A=@n17, B=@n7)) {g=true, r=true})
  n19 = ComponentMask(None=@n13) {b=true}
  n20 = ComponentMask(None=MaterialFunctionCall(In=Abs(None=@n19), Contrast=@n15) {materialFunction=&CheapContrast}.Result) {b=true}
  n21 = ComponentMask(None=LinearInterpolate(
    A=LinearInterpolate(A=@n10.RGB, B=@n12.RGB, Alpha=@n16)
    B=TextureSample(UVs=@n18, Tex=@n9) {texture=&T_FX_LinearSweep_Gradient}.RGB
    Alpha=Saturate(None=@n20)
   )) {r=true}
  n22 = Multiply(
   A=LinearInterpolate(
    A=VectorParameter() {parameterName=Emissive Color A, defaultValue=<rgba 1 0.9995 0.1354 1>}.RGB
    B=VectorParameter() {parameterName=Emissive Color B, defaultValue=<rgba 0 0.7217 1 1>}.RGB
    Alpha=FunctionInput() {inputName=Secondary Mask, inputType=FunctionInput_Scalar}
   )
   B=LinearInterpolate(
    A=ScalarParameter() {defaultValue=0.01, parameterName=Brightness_Min}
    B=ScalarParameter() {defaultValue=8, parameterName=Brightness_Max}
    Alpha=@n21
   )
  )
  n23 = Multiply(A=@n2.BaseColor, B=@n21)
  n24 = FunctionInput() {inputName=Gradient Mask, inputType=FunctionInput_Scalar}
  n25 = ScalarParameter() {parameterName=Reveal}
  n26 = Saturate(None=Subtract(A=Add(A=@n24, B=@n25)))
  n27 = FunctionInput() {inputName=Primary Mask, inputType=FunctionInput_Scalar}
  FunctionOutput(None=BlendMaterialAttributes(
    A=@n1
    B=SetMaterialAttributes!(MaterialAttributes=@n2.MaterialAttributes, EmissiveColor=@n22, BaseColor=@n23) {attributeSetTypes [#2, #1]}
    Alpha=Multiply(A=@n26, B=@n27)
   ))
  SetMaterialAttributes!(MaterialAttributes)
 ]}
