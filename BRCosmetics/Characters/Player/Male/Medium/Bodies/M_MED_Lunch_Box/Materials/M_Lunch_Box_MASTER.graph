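# Master material for the Lunch Box male medium character body skin, a skeletal-mesh material built on Material Attributes.
# The shared character PreFX and PostFX functions bracket the skin-specific effects.
# A Skin FX mask texture feeds primary, secondary and gradient channel-mask parameters into MF_Lunch_Box_Emissive, along with pre-skinned local position.
# That result passes through the weapon-skin MF_EmissiveFresnel function before reaching PostFX.
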Material /BRCosmetics/Characters/Player/Male/Medium/Bodies/M_MED_Lunch_Box/Materials/M_Lunch_Box_MASTER
graph {
 properties {
  bUseMaterialAttributes=true
  bUsedWithMeshDeformer=true
  bUsedWithMorphTargets=true
  bUsedWithSkeletalMesh=true
  pixelDepthOffsetMode=PDOM_Legacy
  usedShadingModels=Default Lit
 }
 nodes [
  ComponentMask() {g=true, r=true}
  n1 = MaterialFunctionCall() {materialFunction=&BaseFN_CharacterFunction_PreFX}
  n2 = TextureSampleParameter2D() {
   parameterName=Skin FX
   samplerType=SAMPLERTYPE_Masks
   texture=/BRCosmetics/Characters/Player/Male/Medium/Bodies/M_MED_Lunch_Box/Textures/T_M_MED_Lunch_Box_Body_FX
  }
  n3 = ChannelMaskParameter(None=@n2.RGBA) {parameterName=Primary Mask}
  n4 = ChannelMaskParameter(None=@n2.RGBA) {parameterName=Secondary Mask}
  n5 = ChannelMaskParameter(None=@n2.RGBA) {maskChannel=Alpha, parameterName=Gradient Emissive Mask, defaultValue=<rgba 0 0 0 1>}
  MaterialFunctionCall(
   AO=@n1.AO
   Distance Blend=@n1.AO
   Fresnel=@n1.AO
   Gmap=@n1.AO
   Pre Skinned Local Position=@n1.AO
   Material=MaterialFunctionCall!(In=MaterialFunctionCall!(
     Primary Mask=@n3.R
     Secondary Mask=@n4.R
     Gradient Mask=@n5.R
     Local Normal
     Local Position=@n1.Pre skinned Local Position
     UVOffset=@n4.R
     Material=@n1.Pre skinned Local Position
    ) {materialFunction=&/MF_Lunch_Box_Emissive}.Result) {materialFunction=/Game/Weapons/WeaponSkins/MaterialFunction/MF_EmissiveFresnel}.Result
  ) {materialFunction=&BaseFN_CharacterFunction_PostFX}
 ]
}
