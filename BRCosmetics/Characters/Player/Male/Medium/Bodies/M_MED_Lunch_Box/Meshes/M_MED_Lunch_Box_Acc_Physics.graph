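# Physics asset for the accessory mesh of the male medium Lunch Box cosmetic body, covering four jacket zipper chains at front and back, left and right.
# Each chain pairs a small capsule on the zipper bone with larger capsule collision on its master bone; front-right master has two capsules.
# One constraint per pair keeps twist and one swing tight while leaving the other swing wide, giving hinge-like motion that ranges further at the front.
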PhysicsAsset /BRCosmetics/Characters/Player/Male/Medium/Bodies/M_MED_Lunch_Box/Meshes/M_MED_Lunch_Box_Acc_Physics
bodies {
 dyn_zipper_jacket_back_l [{shapeType=Capsule, length=2.5, radius=1, center=<2.6239 0 0.0027>, rotation=<rot -89.1425 179.9988 -179.9988>}]
 dyn_zipper_jacket_back_l_master [{shapeType=Capsule, length=7, radius=2.25, center=<1.0274 -0.0002 -3.0154>, rotation=<rot 79.9997 0.0022 0>}]
 dyn_zipper_jacket_back_r [{shapeType=Capsule, length=2.5, radius=1, center=<2.6128 0 0.0022>, rotation=<rot -89.3139 179.9996 -179.9996>}]
 dyn_zipper_jacket_back_r_master [{shapeType=Capsule, length=7.75, radius=2.25, center=<1.1005 0.0001 -2.9071>, rotation=<rot 75.2557 0.0005 -0.0003>}]
 dyn_zipper_jacket_front_l [{shapeType=Capsule, length=1.5, radius=0.75, center=<2.3644 0 0.0687>, rotation=<rot -89.5951 0.0005 -0.0005>}]
 dyn_zipper_jacket_front_l_master [{shapeType=Capsule, length=2, radius=2.25, center=<2.6291 0 -2.9469>, rotation=<rot -89.5757 179.9979 -85.1615>}]
 dyn_zipper_jacket_front_r [{shapeType=Capsule, length=1, radius=0.75, center=<2.2362 0 0.2438>, rotation=<rot -89.7019 -179.9999 179.9999>}]
 dyn_zipper_jacket_front_r_master [
  {shapeType=Capsule, length=2, radius=1.5, center=<2.299 0.0084 -1.6174>, rotation=<rot 0 0 -90.2983>}
  {
   shapeType=Capsule
   length=2.8689
   radius=2.0923
   shapeName=dyn_zipper_jacket_front_r_master_capsule
   center=<0.852 0.0136 -2.6199>
   rotation=<rot 0 0 -90.2983>
  }
 ]
}
constraints [
 {
  bone1Name=dyn_zipper_jacket_back_l
  bone2Name=dyn_zipper_jacket_back_l_master
  swing1LimitDegrees=10
  swing2LimitDegrees=90
  twistLimitDegrees=0.1
 }
 {
  bone1Name=dyn_zipper_jacket_back_r
  bone2Name=dyn_zipper_jacket_back_r_master
  swing1LimitDegrees=10
  swing2LimitDegrees=90
  twistLimitDegrees=0.1
 }
 {
  bone1Name=dyn_zipper_jacket_front_l
  bone2Name=dyn_zipper_jacket_front_l_master
  swing1LimitDegrees=3
  swing2LimitDegrees=120
  twistLimitDegrees=2
 }
 {
  bone1Name=dyn_zipper_jacket_front_r
  bone2Name=dyn_zipper_jacket_front_r_master
  swing1LimitDegrees=3
  swing2LimitDegrees=120
  twistLimitDegrees=2
 }
]
