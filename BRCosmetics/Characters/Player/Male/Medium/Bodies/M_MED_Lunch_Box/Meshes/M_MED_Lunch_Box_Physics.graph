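# Physics asset for the M_MED_Lunch_Box male medium player body, setting up physics for its jacket bones.
# Back, side and front jacket chains of capsules hang from small MSTR anchor spheres through swing- and twist-limited constraints.
# Angularly free cross-links tie each side chain to its front and back neighbors, while the center-back chain stays separate.
# Pelvis, spine, thigh and belt bodies have no constraints, which suggests they act as colliders.
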PhysicsAsset /BRCosmetics/Characters/Player/Male/Medium/Bodies/M_MED_Lunch_Box/Meshes/M_MED_Lunch_Box_Physics
bodies {
 dyn_beltBroch_r [{
   shapeType=Capsule
   length=5.1992
   radius=4.8134
   shapeName=dyn_beltBroch_r_capsule
   center=<5.2795 -1.5379 0.5529>
   rotation=<rot -7.4271 -95.3536 -80.4958>
  }]
 dyn_belt_l [{shapeType=Capsule, length=3.4319, radius=5.9465, shapeName=dyn_belt_r_capsule, center=<-4.0185 1.4758 -1.1804>, rotation=<rot 0 0 89.892>}]
 dyn_belt_r [{shapeType=Capsule, length=4.6912, radius=5.9465, shapeName=dyn_belt_r_capsule, center=<3.9975 0.037 1.3027>, rotation=<rot 0 0 89.892>}]
 dyn_jacket_bk_01 [{shapeType=Capsule, length=7.9266, radius=2.9449, shapeName=dyn_jacket_fr_l_01_capsule, center=<2.7743 -0.3457 2.1736>, rotation=<rot 0 0 90>}]
 dyn_jacket_bk_02 [{shapeType=Capsule, length=7.9266, radius=2.9449, shapeName=dyn_jacket_fr_l_01_capsule, center=<2.7743 -0.3457 1.8302>, rotation=<rot 0 0 90>}]
 dyn_jacket_bk_MSTR [{shapeType=Sphere, radius=0.505, shapeName=dyn_jacket_bk_MSTR_sphere}]
 dyn_jacket_bk_l_01 [{
   shapeType=Capsule
   length=7.437
   radius=3.4346
   shapeName=dyn_jacket_fr_l_01_capsule
   center=<2.8592 -0.1346 1.2604>
   rotation=<rot -0.6414 -2.2502 75.3155>
  }]
 dyn_jacket_bk_l_02 [{
   shapeType=Capsule
   length=7.9266
   radius=2.9449
   shapeName=dyn_jacket_fr_l_01_capsule
   center=<3.8356 -0.384 2.0542>
   rotation=<rot -0.6414 -2.2502 75.3155>
  }]
 dyn_jacket_bk_l_MSTR [{shapeType=Sphere, radius=0.505, shapeName=dyn_jacket_bk_l_MSTR_sphere}]
 dyn_jacket_bk_r_01 [{
   shapeType=Capsule
   length=7.437
   radius=3.6448
   shapeName=dyn_jacket_fr_r_01_capsule
   center=<2.8592 -0.1346 1.2604>
   rotation=<rot -0.6414 2.2502 104.6845>
  }]
 dyn_jacket_bk_r_02 [{
   shapeType=Capsule
   length=7.9266
   radius=2.9449
   shapeName=dyn_jacket_fr_r_01_capsule
   center=<3.8356 -0.384 2.0542>
   rotation=<rot -0.6414 2.2502 104.6845>
  }]
 dyn_jacket_bk_r_MSTR [{shapeType=Sphere, radius=0.505, shapeName=dyn_jacket_bk_r_MSTR_sphere}]
 dyn_jacket_fr_l_02 [{
   shapeType=Capsule
   length=7.9266
   radius=2.9449
   shapeName=dyn_jacket_fr_l_01_capsule
   center=<2.7266 -1.4506 0.7832>
   rotation=<rot -0.1497 0.6924 80.5073>
  }]
 dyn_jacket_fr_l_03 [{
   shapeType=Capsule
   length=7.9266
   radius=2.9449
   shapeName=dyn_jacket_fr_l_01_capsule
   center=<2.7267 -1.4567 0.8028>
   rotation=<rot -0.1479 0.6939 82.0793>
  }]
 dyn_jacket_fr_l_04 [
  {
   shapeType=Capsule
   length=7.9266
   radius=2.9449
   shapeName=dyn_jacket_fr_l_01_capsule
   center=<2.7283 -1.4643 1.3785>
   rotation=<rot -0.1479 0.6939 89.2394>
  }
  {
   shapeType=Capsule
   length=6.2001
   radius=2.7464
   shapeName=dyn_jacket_fr_l_04_capsule
   center=<7.5555 -0.5981 1.6574>
   rotation=<rot 0.0003 0.0002 -88.7855>
  }
 ]
 dyn_jacket_fr_l_MSTR [{shapeType=Sphere, radius=0.505, shapeName=dyn_jacket_fr_l_MSTR_sphere}]
 dyn_jacket_fr_r_02 [{
   shapeType=Capsule
   length=7.9266
   radius=2.9449
   shapeName=dyn_jacket_fr_r_01_capsule
   center=<2.7266 1.4506 0.7832>
   rotation=<rot -0.1497 0.6924 99>
  }]
 dyn_jacket_fr_r_03 [{
   shapeType=Capsule
   length=7.9266
   radius=2.9449
   shapeName=dyn_jacket_fr_r_01_capsule
   center=<2.7277 1.4752 0.9591>
   rotation=<rot -1.0491 5.816 99.1514>
  }]
 dyn_jacket_fr_r_04 [{
   shapeType=Capsule
   length=7.9266
   radius=2.9449
   shapeName=dyn_jacket_fr_r_01_capsule
   center=<3.3134 1.4665 1.0833>
   rotation=<rot -0.5442 5.3938 -86.5966>
  }]
 dyn_jacket_fr_r_MSTR [{shapeType=Sphere, radius=0.505, shapeName=dyn_jacket_fr_r_MSTR_sphere}]
 dyn_jacket_l_01 [{
   shapeType=Capsule
   length=7.9285
   radius=2.943
   shapeName=dyn_jacket_fr_l_01_capsule
   center=<2.2352 -0.4101 1.6689>
   rotation=<rot -0.1497 0.6924 80.5073>
  }]
 dyn_jacket_l_02 [{
   shapeType=Capsule
   length=7.4779
   radius=3.3937
   shapeName=dyn_jacket_fr_l_01_capsule
   center=<3.1846 -0.4675 2.0783>
   rotation=<rot -0.1497 0.6924 80.5073>
  }]
 dyn_jacket_l_03 [{
   shapeType=Capsule
   length=7.9869
   radius=2.8847
   shapeName=dyn_jacket_fr_l_01_capsule
   center=<4.7356 -0.5293 2.5377>
   rotation=<rot -1.0112 -4.4684 83.767>
  }]
 dyn_jacket_l_MSTR [{shapeType=Sphere, radius=0.505, shapeName=dyn_jacket_l_MSTR_sphere}]
 dyn_jacket_r_01 [{
   shapeType=Capsule
   length=7.9285
   radius=2.943
   shapeName=dyn_jacket_fr_l_01_capsule
   center=<2.2352 0.4101 1.6689>
   rotation=<rot -0.1497 0.6923 -84.6964>
  }]
 dyn_jacket_r_02 [{
   shapeType=Capsule
   length=7.4152
   radius=3.4564
   shapeName=dyn_jacket_fr_r_01_capsule
   center=<3.1846 0.4675 2.0783>
   rotation=<rot -0.1497 -0.6924 99.4927>
  }]
 dyn_jacket_r_03 [{
   shapeType=Capsule
   length=8.3869
   radius=2.4846
   shapeName=dyn_jacket_fr_r_01_capsule
   center=<4.7361 0.5353 2.5921>
   rotation=<rot -1.0112 4.471 99.4404>
  }]
 dyn_jacket_r_MSTR [{shapeType=Sphere, radius=0.505, shapeName=dyn_jacket_r_MSTR_sphere}]
 pelvis [{shapeType=Capsule, length=12.1159, radius=9.0172, center=<-0.3896 0 0>}]
 spine_01 [{shapeType=Capsule, length=7.5419, radius=9.7582, center=<-0.7795 0.9827 0>}]
 spine_02 [{shapeType=Capsule, length=0, radius=10.4732, center=<-0.6238 1.7349 0.0078>, rotation=<rot -89.2831 179.9995 -174.8198>}]
 spine_03 [{shapeType=Capsule, length=4.7857, radius=9.7148, center=<-0.6238 1.7349 0.0078>, rotation=<rot 0.9006 -174.8283 -179.9353>}]
 thigh_l [
  {shapeType=Capsule, length=21.6728, radius=8.1052, center=<-12.4612 0.3845 0.1823>, rotation=<rot 90 0 0>}
  {
   shapeType=Capsule
   length=12.2045
   radius=7.3284
   shapeName=thigh_l_capsule
   center=<-9.3964 1.9607 6.7533>
   rotation=<rot 81.399 83.0539 82.6198>
  }
 ]
 thigh_r [{shapeType=Capsule, length=21.6728, radius=8.1052, center=<12.4612 -0.3845 -0.1823>, rotation=<rot 90 0 180>}]
}
constraints [
 {bone1Name=dyn_jacket_bk_01, bone2Name=dyn_jacket_bk_MSTR, swing1LimitDegrees=5, swing2LimitDegrees=20, twistLimitDegrees=2}
 {bone1Name=dyn_jacket_bk_l_01, bone2Name=dyn_jacket_bk_l_MSTR, swing1LimitDegrees=10, swing2LimitDegrees=20, twistLimitDegrees=5}
 {bone1Name=dyn_jacket_l_01, bone2Name=dyn_jacket_l_MSTR, swing1LimitDegrees=10, swing2LimitDegrees=40, twistLimitDegrees=5}
 {bone1Name=dyn_jacket_fr_l_03, bone2Name=dyn_jacket_fr_l_02, swing1LimitDegrees=5, swing2LimitDegrees=30, twistLimitDegrees=2}
 {bone1Name=dyn_jacket_fr_l_04, bone2Name=dyn_jacket_fr_l_03, swing1LimitDegrees=5, swing2LimitDegrees=30, twistLimitDegrees=2}
 {bone1Name=dyn_jacket_l_02, bone2Name=dyn_jacket_l_01, swing1LimitDegrees=10, swing2LimitDegrees=40, twistLimitDegrees=5}
 {bone1Name=dyn_jacket_l_03, bone2Name=dyn_jacket_l_02, swing1LimitDegrees=10, swing2LimitDegrees=40, twistLimitDegrees=5}
 {bone1Name=dyn_jacket_bk_l_02, bone2Name=dyn_jacket_bk_l_01, swing1LimitDegrees=5, swing2LimitDegrees=15, twistLimitDegrees=5}
 {bone1Name=dyn_jacket_bk_02, bone2Name=dyn_jacket_bk_01, swing1LimitDegrees=5, swing2LimitDegrees=20, twistLimitDegrees=2}
 {
  bone1Name=dyn_jacket_fr_l_02
  bone2Name=dyn_jacket_l_01
  swing1LimitDegrees=10
  swing1Motion=Free
  swing2LimitDegrees=40
  swing2Motion=Free
  twistLimitDegrees=5
  twistMotion=Free
 }
 {bone1Name=dyn_jacket_fr_r_03, bone2Name=dyn_jacket_fr_r_02, swing1LimitDegrees=10, swing2LimitDegrees=30, twistLimitDegrees=5}
 {bone1Name=dyn_jacket_fr_r_04, bone2Name=dyn_jacket_fr_r_03, swing1LimitDegrees=10, swing2LimitDegrees=30, twistLimitDegrees=5}
 {bone1Name=dyn_jacket_r_01, bone2Name=dyn_jacket_r_MSTR, swing1LimitDegrees=10, swing2LimitDegrees=30, twistLimitDegrees=5}
 {bone1Name=dyn_jacket_r_02, bone2Name=dyn_jacket_r_01, swing1LimitDegrees=10, swing2LimitDegrees=40, twistLimitDegrees=5}
 {bone1Name=dyn_jacket_r_03, bone2Name=dyn_jacket_r_02, swing1LimitDegrees=10, swing2LimitDegrees=40, twistLimitDegrees=5}
 {
  bone1Name=dyn_jacket_fr_r_02
  bone2Name=dyn_jacket_r_01
  swing1LimitDegrees=45
  swing1Motion=Free
  swing2LimitDegrees=45
  swing2Motion=Free
  twistLimitDegrees=45
  twistMotion=Free
 }
 {bone1Name=dyn_jacket_bk_r_01, bone2Name=dyn_jacket_bk_r_MSTR, swing1LimitDegrees=10, swing2LimitDegrees=20, twistLimitDegrees=5}
 {bone1Name=dyn_jacket_bk_r_02, bone2Name=dyn_jacket_bk_r_01, swing1LimitDegrees=5, swing2LimitDegrees=15, twistLimitDegrees=5}
 {bone1Name=dyn_jacket_fr_l_02, bone2Name=dyn_jacket_fr_l_MSTR, swing1LimitDegrees=10, swing2LimitDegrees=40, twistLimitDegrees=5}
 {bone1Name=dyn_jacket_fr_r_02, bone2Name=dyn_jacket_fr_r_MSTR, swing1LimitDegrees=10, swing2LimitDegrees=40, twistLimitDegrees=5}
 {
  bone1Name=dyn_jacket_l_03
  bone2Name=dyn_jacket_fr_l_04
  swing1LimitDegrees=45
  swing1Motion=Free
  swing2LimitDegrees=45
  swing2Motion=Free
  twistLimitDegrees=45
  twistMotion=Free
 }
 {
  bone1Name=dyn_jacket_r_03
  bone2Name=dyn_jacket_fr_r_04
  swing1LimitDegrees=45
  swing1Motion=Free
  swing2LimitDegrees=45
  swing2Motion=Free
  twistLimitDegrees=45
  twistMotion=Free
 }
 {
  bone1Name=dyn_jacket_bk_r_02
  bone2Name=dyn_jacket_r_03
  swing1LimitDegrees=45
  swing1Motion=Free
  swing2LimitDegrees=45
  swing2Motion=Free
  twistLimitDegrees=45
  twistMotion=Free
 }
 {
  bone1Name=dyn_jacket_bk_l_02
  bone2Name=dyn_jacket_l_03
  swing1LimitDegrees=45
  swing1Motion=Free
  swing2LimitDegrees=45
  swing2Motion=Free
  twistLimitDegrees=45
  twistMotion=Free
 }
]
