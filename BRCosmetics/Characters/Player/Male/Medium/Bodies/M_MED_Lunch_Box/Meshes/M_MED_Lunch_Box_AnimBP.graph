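# Post-process-style AnimBP for the male medium Lunch Box cosmetic body, adding secondary motion to jacket, hood, belt and back flap.
# The parent mesh pose feeds a Control Rig, weighted bone constraints, then AnimDynamics and LookAt nodes.
# A final RigidBody pass selects a gameplay or front-end physics asset.
# Update events set state flags and skydive-aware gravity; shoe-flap nodes and the back-flap constraint are unconnected.
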AnimBlueprint /BRCosmetics/Characters/Player/Male/Medium/Bodies/M_MED_Lunch_Box/Meshes/M_MED_Lunch_Box_AnimBP
target_skeleton=&/M_MED_Lunch_Box_Skeleton
graphs [
 AnimGraph {nodes [
   n1 = Anim Dynamics - Bone: dyn_l_shoe_flap(Alpha=0, ExternalForce="0,0,0") {
    angularDampingOverride=0.99
    angularSpringConstant=10
    bAngularSpring=true
    bOverrideAngularDamping=true
    bUseGravityOverride=true
    simulationSpace=RootRelative
    alphaBoolBlend {blendOption=Linear}
    boundBone {boneName=dyn_l_shoe_flap}
    componentAppliedLinearAccClamp=<3000 3000 3000>
    componentLinearAccScale=<0.3 0.3 0.3>
    componentLinearVelScale=<0.3 0.3 0.3>
    gravityOverride {z=-980}
    physicsBodyDefinitions [{
      boundBone {boneName=dyn_l_shoe_flap}
      boxExtents=<20 20 10>
      constraintSetup {angularLimitsMax=<0 10 0>, angularLimitsMin=<0 -3 0>, angularTarget=<0 0 0>}
      localJointOffset=<1 0 0>
     }]
   }
   Anim Dynamics - Bone: dyn_r_shoe_flap(ComponentPose=@n1, Alpha=0, ExternalForce="0,0,0") {
    angularDampingOverride=0.99
    angularSpringConstant=10
    bAngularSpring=true
    bOverrideAngularDamping=true
    bUseGravityOverride=true
    simulationSpace=RootRelative
    alphaBoolBlend {blendOption=Linear}
    boundBone {boneName=dyn_r_shoe_flap}
    componentAppliedLinearAccClamp=<3000 3000 3000>
    componentLinearAccScale=<0.3 0.3 0.3>
    componentLinearVelScale=<0.3 0.3 0.3>
    gravityOverride {z=-980}
    physicsBodyDefinitions [{
      boundBone {boneName=dyn_r_shoe_flap}
      boxExtents=<20 20 10>
      constraintSetup {angularLimitsMax=<0 10 0>, angularLimitsMin=<0 -3 0>, angularTarget=<0 0 0>}
      localJointOffset=<1 0 0>
     }]
   }
   "Constraint - dyn_back_Flap_master (PPPPP)"!(
    ComponentPose
    Alpha=1
    ConstraintWeights_0=0.7567
    ConstraintWeights_1:float=0.0636
    ConstraintWeights_2:float=0.0636
    ConstraintWeights_3:float=0.058
    ConstraintWeights_4:float=0.058
   ) -> (Pose) {
    alphaBoolBlend {blendOption=Linear}
    boneToModify {boneName=dyn_back_Flap_master}
    constraintSetup [
     {targetBone {boneName=pelvis}}
     {targetBone {boneName=thigh_twist_01_l}}
     {targetBone {boneName=thigh_twist_01_r}}
     {targetBone {boneName=deform_glute_l}}
     {targetBone {boneName=deform_glute_r}}
    ]
    constraintWeights [1,1,1,1,1]
   }
   n2 = Look At - Bone: dyn_l_shoe_flap(Alpha=1) {
    alphaBoolBlend {blendOption=Linear}
    boneToModify {boneName=dyn_l_shoe_flap}
    lookAtLocation=<14.028 7.9612 -6.9479>
    lookAtTarget {boneReference {boneName=calf_l}}
    lookAt_Axis {axis=<-1 0>}
    lookUp_Axis {axis=<0 1>}
   }
   Look At - Bone: dyn_r_shoe_flap(ComponentPose=@n2, Alpha=1) {
    alphaBoolBlend {blendOption=Linear}
    boneToModify {boneName=dyn_r_shoe_flap}
    lookAtLocation=<-16.2304 -8.1064 4.3719>
    lookAtTarget {boneReference {boneName=calf_r}}
    lookAt_Axis {axis=<-1 0>}
    lookUp_Axis {axis=<0 1>}
   }
   n3 = RigidBody!(ComponentPose=Local To Component(LocalPose=Use cached pose 'AnimDynamics'()), Alpha, OverrideWorldGravity:Vector="0,0,0", ExternalForce) -> (Pose) {
    bClampLinearTranslationLimitToRefPose=true
    bDefaultToSkeletalMeshPhysicsAsset=false
    bOverrideWorldGravity=true
    overridePhysicsAsset=&/M_MED_Lunch_Box_Frontend_Physics
    simulationSpace=BaseBoneSpace
    alphaBoolBlend {blendOption=Linear}
    baseBoneRef {boneName=pelvis}
    simSpaceSettings {
     dampingAlpha=0
     maxAngularAcceleration=40
     maxAngularVelocity=10
     maxLinearAcceleration=500
     maxLinearVelocity=500
     velocityScaleZ=0.7
     worldAlpha=0.3
     externalLinearDragV=<0.5 0.5 0.5>
    }
   }
   n4 = RigidBody(ComponentPose=Local To Component(LocalPose=Use cached pose 'AnimDynamics'()), OverrideWorldGravity="0,0,0") {
    bClampLinearTranslationLimitToRefPose=true
    bDefaultToSkeletalMeshPhysicsAsset=false
    bOverrideWorldGravity=true
    overridePhysicsAsset=&/M_MED_Lunch_Box_Physics
    simulationSpace=BaseBoneSpace
    alphaBoolBlend {blendOption=Linear}
    baseBoneRef {boneName=pelvis}
    simSpaceSettings {
     dampingAlpha=0
     maxAngularAcceleration=40
     maxAngularVelocity=10
     maxLinearAcceleration=500
     maxLinearVelocity=500
     velocityScaleZ=0.7
     worldAlpha=0.3
     externalLinearDragV=<0.5 0.5 0.5>
    }
   }
   Output Pose(Result=Blend Poses by bool(
     BlendPose_0=Component To Local(ComponentPose=@n4)
     BlendPose_1=Component To Local(ComponentPose=@n3)
     bActiveValue=Get NotFrontEnd() -> (NotFrontEnd:bool)
    ) {blendType=Linear}) {layerGroup=None}
   n5 = Control Rig(Source=Copy Pose From Mesh() {bUseAttachedParent=true}) {
    alphaBoolBlend {blendOption=Linear}
    controlRigAssetReference {blueprintRigClass=&/M_MED_Lunch_Box_CtrlRig.M_MED_Lunch_Box_CtrlRig_C}
   }
   n6 = "Constraint - dyn_zipper_jacket_front_r_master (PPPPPPPP)"!(
    ComponentPose=Local To Component(LocalPose=@n5)
    Alpha=1
    ConstraintWeights_0=0.0055
    ConstraintWeights_1:float=0.1129
    ConstraintWeights_2:float=0.416
    ConstraintWeights_3:float=0.408
    ConstraintWeights_4:float=0.0017
    ConstraintWeights_5:float=0.0082
    ConstraintWeights_6:float=0.046
    ConstraintWeights_7:float=0.0017
   ) -> (Pose) {
    alphaBoolBlend {blendOption=Linear}
    boneToModify {boneName=dyn_zipper_jacket_front_r_master}
    constraintSetup [
     {targetBone {boneName=spine_02}}
     {targetBone {boneName=spine_03}}
     {targetBone {boneName=spine_04}}
     {targetBone {boneName=spine_05}}
     {targetBone {boneName=neck_01}}
     {targetBone {boneName=neck_02}}
     {targetBone {boneName=clavicle_l}}
     {targetBone {boneName=upperarm_twist_01_l}}
    ]
    constraintWeights [1,1,1,1,1,1,1,1]
   }
   n7 = "Constraint - dyn_zipper_jacket_front_l_master (PPPPPPPP)"!(
    ComponentPose=@n6
    Alpha=1
    ConstraintWeights_0=0.0064
    ConstraintWeights_1:float=0.1151
    ConstraintWeights_2:float=0.4251
    ConstraintWeights_3:float=0.4011
    ConstraintWeights_4:float=0.0016
    ConstraintWeights_5:float=0.0076
    ConstraintWeights_6:float=0.0417
    ConstraintWeights_7:float=0.0015
   ) -> (Pose) {
    alphaBoolBlend {blendOption=Linear}
    boneToModify {boneName=dyn_zipper_jacket_front_l_master}
    constraintSetup [
     {targetBone {boneName=spine_02}}
     {targetBone {boneName=spine_03}}
     {targetBone {boneName=spine_04}}
     {targetBone {boneName=spine_05}}
     {targetBone {boneName=neck_01}}
     {targetBone {boneName=neck_02}}
     {targetBone {boneName=clavicle_l}}
     {targetBone {boneName=upperarm_twist_01_l}}
    ]
    constraintWeights [1,1,1,1,1,1,1,1]
   }
   n8 = "Constraint - dyn_zipper_jacket_back_r_master (PPPPPPPP)"!(
    ComponentPose=@n7
    Alpha=1
    ConstraintWeights_0=0.0099
    ConstraintWeights_1:float=0.194
    ConstraintWeights_2:float=0.2713
    ConstraintWeights_3:float=0.4232
    ConstraintWeights_4:float=0.0036
    ConstraintWeights_5:float=0.083
    ConstraintWeights_6:float=0.0009
    ConstraintWeights_7:float=0.014
   ) -> (Pose) {
    alphaBoolBlend {blendOption=Linear}
    boneToModify {boneName=dyn_zipper_jacket_back_r_master}
    constraintSetup [
     {targetBone {boneName=spine_02}}
     {targetBone {boneName=spine_03}}
     {targetBone {boneName=spine_04}}
     {targetBone {boneName=spine_05}}
     {targetBone {boneName=neck_02}}
     {targetBone {boneName=clavicle_r}}
     {targetBone {boneName=upperarm_r}}
     {targetBone {boneName=upperarm_twist_01_r}}
    ]
    constraintWeights [1,1,1,1,1,1,1,1]
   }
   n9 = "Constraint - dyn_zipper_jacket_back_l_master (PPPPPPPP)"!(
    ComponentPose=@n8
    Alpha=1
    ConstraintWeights_0=0.0134
    ConstraintWeights_1:float=0.2022
    ConstraintWeights_2:float=0.2731
    ConstraintWeights_3:float=0.4152
    ConstraintWeights_4:float=0.0002
    ConstraintWeights_5:float=0.0024
    ConstraintWeights_6:float=0.0928
    ConstraintWeights_7:float=0.0007
   ) -> (Pose) {
    alphaBoolBlend {blendOption=Linear}
    boneToModify {boneName=dyn_zipper_jacket_back_l_master}
    constraintSetup [
     {targetBone {boneName=spine_02}}
     {targetBone {boneName=spine_03}}
     {targetBone {boneName=spine_04}}
     {targetBone {boneName=spine_05}}
     {targetBone {boneName=neck_01}}
     {targetBone {boneName=neck_02}}
     {targetBone {boneName=clavicle_l}}
     {targetBone {boneName=upperarm_twist_01_l}}
    ]
    constraintWeights [1,1,1,1,1,1,1,1]
   }
   n10 = "Constraint - dyn_beltBroch_l (PPPP)"!(
    ComponentPose=@n9
    Alpha=1
    ConstraintWeights_0=0.2946
    ConstraintWeights_1:float=0.2669
    ConstraintWeights_2:float=0.4377
    ConstraintWeights_3:float=0.0007
   ) -> (Pose) {
    alphaBoolBlend {blendOption=Linear}
    boneToModify {boneName=dyn_beltBroch_l}
    constraintSetup [
     {targetBone {boneName=pelvis}}
     {targetBone {boneName=thigh_twist_01_r}}
     {targetBone {boneName=deform_glute_r}}
     {targetBone {boneName=deform_groin_r}}
    ]
    constraintWeights [1,1,1,1]
   }
   n11 = "Constraint - dyn_beltBroch_r (PPPPPP)"!(
    ComponentPose=@n10
    Alpha=0
    ConstraintWeights_0=0.2738
    ConstraintWeights_1:float=0.0015
    ConstraintWeights_2:float=0.3959
    ConstraintWeights_3:float=0.0102
    ConstraintWeights_4:float=0.3174
    ConstraintWeights_5:float=0.0011
   ) -> (Pose) {
    alphaBoolBlend {blendOption=Linear}
    boneToModify {boneName=dyn_beltBroch_r}
    constraintSetup [
     {targetBone {boneName=pelvis}}
     {targetBone {boneName=thigh_l}}
     {targetBone {boneName=thigh_twist_01_l}}
     {targetBone {boneName=thigh_twist_01_r}}
     {targetBone {boneName=deform_groin_l}}
     {targetBone {boneName=deform_groin_r}}
    ]
    constraintWeights [1,1,1,1,1,1]
   }
   n12 = "Constraint - dyn_jacket_zipper_front_MSTR (PPPPPPP)"!(
    ComponentPose=@n11
    bAlphaBoolEnabled:bool=Get IsRBANOff() -> (IsRBANOff:bool)
    ConstraintWeights_0=0.5885
    ConstraintWeights_1:float=0.0465
    ConstraintWeights_2:float=0.0342
    ConstraintWeights_3:float=0.0177
    ConstraintWeights_4:float=0.0622
    ConstraintWeights_5:float=0.0081
    ConstraintWeights_6:float=0.2428
   ) -> (Pose) {
    alphaInputType=Bool
    alphaBoolBlend {blendOption=Linear}
    boneToModify {boneName=dyn_jacket_zipper_front_MSTR}
    constraintSetup [
     {targetBone {boneName=pelvis}}
     {targetBone {boneName=spine_01}}
     {targetBone {boneName=spine_02}}
     {targetBone {boneName=spine_03}}
     {targetBone {boneName=thigh_twist_01_l}}
     {targetBone {boneName=thigh_twist_01_r}}
     {targetBone {boneName=deform_groin_l}}
    ]
    constraintWeights [1,1,1,1,1,1,1]
   }
   Save cached pose 'PreSetup'(Pose=Component To Local(ComponentPose=@n12)) {cacheName=PreSetup}
   n13 = Anim Dynamics - Bone: dyn_belt_look_l(ComponentPose=Local To Component(LocalPose=Use cached pose 'PreSetup'()), Alpha=0.5, ExternalForce="0,0,0") {
    angularDampingOverride=0.99
    bAngularSpring=true
    bOverrideAngularDamping=true
    bUseGravityOverride=true
    simulationSpace=RootRelative
    alphaBoolBlend {blendOption=Linear}
    boundBone {boneName=dyn_belt_look_l}
    componentAppliedLinearAccClamp=<3000 3000 3000>
    componentLinearAccScale=<1 1 1>
    componentLinearVelScale=<1 1 1>
    gravityOverride {z=-980}
    physicsBodyDefinitions [{
      boundBone {boneName=dyn_belt_look_l}
      boxExtents=<20 20 10>
      constraintSetup {angularLimitsMax=<0 0 0>, angularLimitsMin=<0 -20 0>, angularTarget=<1 0 0>}
      localJointOffset=<1 0 0>
     }]
   }
   n14 = Anim Dynamics - Bone: dyn_belt_rook_r(ComponentPose=@n13, Alpha=0.5, ExternalForce="0,0,0") {
    angularDampingOverride=0.99
    bAngularSpring=true
    bOverrideAngularDamping=true
    bUseGravityOverride=true
    simulationSpace=RootRelative
    alphaBoolBlend {blendOption=Linear}
    boundBone {boneName=dyn_belt_rook_r}
    componentAppliedLinearAccClamp=<3000 3000 3000>
    componentLinearAccScale=<1 1 1>
    componentLinearVelScale=<1 1 1>
    gravityOverride {z=-980}
    physicsBodyDefinitions [{
      boundBone {boneName=dyn_belt_rook_r}
      boxExtents=<20 20 10>
      constraintSetup {angularLimitsMax=<2 2 2>, angularLimitsMin=<-2 -20 -2>, angularTarget=<1 0 0>}
      localJointOffset=<-1 0 0>
     }]
   }
   n15 = Look At - Bone: dyn_beltBroch_l(ComponentPose=@n14, Alpha=1) {
    alphaBoolBlend {blendOption=Linear}
    boneToModify {boneName=dyn_beltBroch_l}
    lookAtLocation=<-6.5926 3.715 -0.4614>
    lookAtTarget {boneReference {boneName=dyn_belt_r}}
    lookAt_Axis {axis=<1 0>}
    lookUp_Axis {axis {x=0, z=1}}
   }
   n16 = Look At - Bone: dyn_beltBroch_r(ComponentPose=@n15, Alpha=1) {
    alphaBoolBlend {blendOption=Linear}
    boneToModify {boneName=dyn_beltBroch_r}
    lookAtLocation=<-2.5222 5.5326 -1.2619>
    lookAtTarget {boneReference {boneName=dyn_belt_l}}
    lookAt_Axis {axis=<1 0>}
    lookUp_Axis {axis {x=0, z=1}}
   }
   n17 = Anim Dynamics - Bone: dyn_zipper_jacket_back_l(ComponentPose=@n16, Alpha=0.8, ExternalForce="0,0,0") {
    angularDampingOverride=0.99
    angularSpringConstant=50
    bAngularSpring=true
    bOverrideAngularDamping=true
    bUseGravityOverride=true
    simulationSpace=RootRelative
    alphaBoolBlend {blendOption=Linear}
    boundBone {boneName=dyn_zipper_jacket_back_l}
    componentAppliedLinearAccClamp=<3000 3000 3000>
    componentLinearAccScale=<1.1 1.1 1.1>
    componentLinearVelScale=<1.1 1.1 1.1>
    gravityOverride {z=-980}
    physicsBodyDefinitions [{
      boundBone {boneName=dyn_zipper_jacket_back_l}
      boxExtents=<15 15 15>
      constraintSetup {angularLimitsMax=<0 2 30>, angularLimitsMin=<0 -110 -30>, angularTarget=<1 0 0>}
      localJointOffset=<0.9 0 0>
     }]
   }
   n18 = Anim Dynamics - Bone: dyn_zipper_jacket_back_r(ComponentPose=@n17, Alpha=1, ExternalForce="0,0,0") {
    angularDampingOverride=0.95
    angularSpringConstant=25
    bAngularSpring=true
    bOverrideAngularDamping=true
    bUseGravityOverride=true
    simulationSpace=RootRelative
    alphaBoolBlend {blendOption=Linear}
    boundBone {boneName=dyn_zipper_jacket_back_r}
    componentAppliedLinearAccClamp=<3500 3500 3500>
    componentLinearAccScale=<1 1 1>
    componentLinearVelScale=<1 1 1>
    gravityOverride {z=-980}
    physicsBodyDefinitions [{
      boundBone {boneName=dyn_zipper_jacket_back_r}
      boxExtents=<14 14 14>
      constraintSetup {angularLimitsMax=<0 2 30>, angularLimitsMin=<0 -100 -30>, angularTarget=<1 0 0>}
      localJointOffset=<0.8 0 0>
     }]
    relativeSpaceBone {boneName=spine_04}
   }
   n19 = Anim Dynamics - Bone: dyn_zipper_jacket_front_l(ComponentPose=@n18, Alpha=1, ExternalForce="0,0,0") {
    angularDampingOverride=0.99
    angularSpringConstant=25
    bAngularSpring=true
    bOverrideAngularDamping=true
    bUseGravityOverride=true
    simulationSpace=RootRelative
    alphaBoolBlend {blendOption=Linear}
    boundBone {boneName=dyn_zipper_jacket_front_l}
    componentAppliedLinearAccClamp=<2900 2900 2900>
    componentLinearAccScale=<1 1 1>
    componentLinearVelScale=<1 1 1>
    gravityOverride {z=-980}
    physicsBodyDefinitions [{
      boundBone {boneName=dyn_zipper_jacket_front_l}
      boxExtents=<15 15 15>
      constraintSetup {angularLimitsMax=<0 2 10>, angularLimitsMin=<0 -110 -10>, angularTarget=<1 0 0>}
      localJointOffset=<0.7 0 0>
     }]
   }
   n20 = Anim Dynamics - Bone: dyn_zipper_jacket_front_r(ComponentPose=@n19, Alpha=0.8, ExternalForce="0,0,0") {
    angularDampingOverride=0.95
    angularSpringConstant=25
    bAngularSpring=true
    bOverrideAngularDamping=true
    bUseGravityOverride=true
    simulationSpace=RootRelative
    alphaBoolBlend {blendOption=Linear}
    boundBone {boneName=dyn_zipper_jacket_front_r}
    componentAppliedLinearAccClamp=<2800 2800 2800>
    componentLinearAccScale=<1.2 1.2 1.2>
    componentLinearVelScale=<1.2 1.2 1.2>
    gravityOverride {z=-980}
    physicsBodyDefinitions [{
      boundBone {boneName=dyn_zipper_jacket_front_r}
      boxExtents=<14 14 14>
      constraintSetup {angularLimitsMax=<0 2 10>, angularLimitsMin=<0 -100 -10>, angularTarget=<1 0 0>}
      localJointOffset=<0.8 0 0>
     }]
   }
   n21 = Anim Dynamics - Bone: dyn_back_Flap(ComponentPose=@n20, Alpha=1, ExternalForce="0,0,0") {
    angularDampingOverride=0.99
    angularSpringConstant=25
    bAngularSpring=true
    bOverrideAngularDamping=true
    bUseGravityOverride=true
    simulationSpace=RootRelative
    alphaBoolBlend {blendOption=Linear}
    boundBone {boneName=dyn_back_Flap}
    componentAppliedLinearAccClamp=<3000 3000 3000>
    componentLinearAccScale=<0.3 0.3 0.3>
    componentLinearVelScale=<0.3 0.3 0.3>
    gravityOverride {z=-980}
    physicsBodyDefinitions [{
      boundBone {boneName=dyn_back_Flap}
      boxExtents=<20 20 10>
      constraintSetup {angularLimitsMax=<0 0 0>, angularLimitsMin=<0 -15 0>, angularTarget=<1 0 0>}
      localJointOffset=<1 0 0>
     }]
   }
   n22 = Anim Dynamics - Bone: dyn_hood(ComponentPose=@n21, Alpha=1, ExternalForce="0,0,0") {
    angularDampingOverride=0.95
    angularSpringConstant=10
    bAngularSpring=true
    bOverrideAngularDamping=true
    bOverrideLinearDamping=true
    bUseGravityOverride=true
    linearDampingOverride=0.85
    simulationSpace=RootRelative
    alphaBoolBlend {blendOption=Linear}
    boundBone {boneName=dyn_hood}
    componentAppliedLinearAccClamp=<3000 3000 3000>
    componentLinearAccScale=<0.3 0.3 0.3>
    componentLinearVelScale=<0.3 0.3 0.3>
    gravityOverride {z=-980}
    physicsBodyDefinitions [{
      boundBone {boneName=dyn_hood}
      boxExtents=<20 20 20>
      constraintSetup {
       angularTargetAxis=AxisZ
       twistAxis=AxisZ
       angularLimitsMax=<0 10 0>
       angularLimitsMin=<-5 -10 0>
       angularTarget=<0 0 1>
       linearAxesMax=<0.5 0 0>
       linearAxesMin=<-0.5 0 -1>
      }
      localJointOffset=<0 0 -1>
     }]
   }
   n23 = Anim Dynamics - Bone: dyn_hoodieBottom(ComponentPose=@n22, Alpha=1, ExternalForce="0,0,0") {
    angularDampingOverride=0.95
    angularSpringConstant=25
    bAngularSpring=true
    bOverrideAngularDamping=true
    bUseGravityOverride=true
    linearDampingOverride=0.85
    simulationSpace=RootRelative
    alphaBoolBlend {blendOption=Linear}
    boundBone {boneName=dyn_hoodieBottom}
    componentAppliedLinearAccClamp=<3000 3000 3000>
    componentLinearAccScale=<0.5 0.5 0.5>
    componentLinearVelScale=<0.5 0.5 0.5>
    gravityOverride {z=-980}
    physicsBodyDefinitions [{
      boundBone {boneName=dyn_hoodieBottom}
      boxExtents=<10 20 20>
      constraintSetup {angularLimitsMax=<0 2 5>, angularLimitsMin=<0 -30 -5>, angularTarget=<1 0 0>, linearAxesMax=<0.5 0 0>}
      localJointOffset=<1 0 0>
     }]
    relativeSpaceBone {boneName=spine_05}
   }
   Save cached pose 'AnimDynamics'(Pose=Component To Local(ComponentPose=@n23)) {cacheName=AnimDynamics}
   Get NonCrouch() -> (NonCrouch:double)
  ]}
 EventGraph {nodes [
   Event BlueprintUpdateAnimation()
   > n1 = Gravity Override State Function DEPRECATED(CharacterPartAnimInstance=Self-Reference(), GameTimeInSeconds=GetGameTimeInSeconds(), GravityOverrideParams=Get Gravity Override Params())
   > Set GravityOverride(GravityOverride=@n1)
   > Set IsRBANOff(IsRBANOff:bool=NOT Boolean(A=Get bRigidBodyEnabled())) -> (Output_Get:bool)
   n2 = NOT Boolean(A=Get bIsCrouching())
   > Set NonCrouch(NonCrouch:double=Remap Float(Input="To Float (Boolean)"(InBool=@n2), OldRange=(X=0,Y=1), NewRange=(X=0.5,Y=1), Clamp=true) -> (Output=~)) -> (Output_Get:double)
   > Set NotFrontEnd(NotFrontEnd:bool=NOT Boolean(A=Get bIsInFrontEnd())) -> (Output_Get:bool)
   n3 = BreakTransform(InTransform=GetDeltaTransformFromRefPose(self=GetOwnerSkeletalMeshComponent(), BoneName=thigh_r, BaseName=pelvis))
   Set Thigh_Step(Thigh_Step:double=MapRangeClamped(Value=BreakRotator(InRot=@n3.Rotation).Yaw, InRangeA=-100, InRangeB=-40, OutRangeA=1, OutRangeB=0)) -> (Output_Get:double)
   n4 = GetOwnerSkeletalMeshComponent()
   n5 = BreakRotator(InRot=BreakTransform(InTransform=GetDeltaTransformFromRefPose(self=@n4, BoneName=thigh_l, BaseName=pelvis)).Rotation)
   > Set Thigh_Step_R(Thigh_Step_R:double=MapRangeClamped(Value=@n5.Yaw, InRangeA=-100, InRangeB=-40)) -> (Output_Get:double)
   > Set Incline_Step_L(Incline_Step_L:double=MapRangeClamped(Value=@n5.Yaw, InRangeA=-120, InRangeB=-140)) -> (Output_Get:double)
   n6 = BreakTransform(InTransform=GetDeltaTransformFromRefPose(self=GetOwnerSkeletalMeshComponent(), BoneName=upperarm_r, BaseName=clavicle_r))
   > Set Head_Tilt(Head_Tilt:double=MapRangeClamped(Value=BreakRotator(InRot=@n6.Rotation).Yaw, InRangeA=95, InRangeB=120)) -> (Output_Get:double)
   "To String (Float)"(InDouble=@n5.Yaw)
   Set Is Sky Diving(Is Sky Diving:double="To Float (Boolean)"(InBool=Get bIsSkyDiving())) -> (Output_Get:double)
  ]}
]
variables [
 Gravity Override Params {
  type=GravityOverrideParamsStruct
  default {
   skydive_FloatingDown {gravityOverride=<0 -1200 -600>, windAmplitude=<200 300 100>, windFrequency=<1.5 5 8>, windOffset=<0 -500 400>}
   skydive_GoingUp {gravityOverride=<0 -1000 0>, windAmplitude=<200 300 300>, windFrequency=<1.5 5 8>}
   skydive_Parachuting {gravityOverride=<0 -300 -980>, windAmplitude=<200 -100 100>}
  }
 }
 GravityOverride {type=Vector, default=<0 0 -980>}
 Head_Tilt {type=number}
 Incline_Step_L {type=number}
 Is Sky Diving {type=number}
 IsRBANOff {type=boolean}
 NonCrouch {type=number}
 NotFrontEnd {type=boolean}
 Thigh_Step {type=number}
 Thigh_Step_R {type=number}
]
