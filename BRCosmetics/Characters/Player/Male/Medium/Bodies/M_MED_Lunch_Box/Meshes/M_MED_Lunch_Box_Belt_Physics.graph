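# Physics asset for the belt of the M_MED_Lunch_Box male player body, defining capsule bodies and joints for its dangling chains and flaps.
# Four five-link chains hang from master bodies angularly locked to the pelvis, with each side's front and back chains cross-linked near the ends by tight swing limits.
# A front flap branches into left and right segments, while the back flap's master swings freely.
# Pelvis and multi-capsule thighs supply body shapes.
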PhysicsAsset /BRCosmetics/Characters/Player/Male/Medium/Bodies/M_MED_Lunch_Box/Meshes/M_MED_Lunch_Box_Belt_Physics
bodies {
 dyn_back_Flap [{shapeType=Capsule, length=4.25, radius=1.75, center=<4.0706 0.0001 -0.0483>, rotation=<rot -89.2551 -179.9995 179.9995>}]
 dyn_back_Flap_master [{shapeType=Capsule, length=1, radius=1}]
 dyn_front_flap [{shapeType=Capsule, length=1, radius=1}]
 dyn_front_flap_l_01 [{shapeType=Capsule, length=4, radius=2, center=<3.5 0 0>, rotation=<rot 90 0 0>}]
 dyn_front_flap_l_02 [{shapeType=Capsule, length=4, radius=2, center=<3.5 0 0>, rotation=<rot 90 0 0>}]
 dyn_front_flap_r_01 [{shapeType=Capsule, length=4, radius=2, center=<3.5 0 0>, rotation=<rot 90 0 0>}]
 dyn_front_flap_r_02 [{shapeType=Capsule, length=4, radius=2, center=<3.5 0 0>, rotation=<rot 90 0 0>}]
 dyn_l_back_chain_01 [{shapeType=Capsule, length=2.5, radius=2, rotation=<rot 90 0 0>}]
 dyn_l_back_chain_02 [{shapeType=Capsule, length=2.5, radius=2, rotation=<rot 90 0 0>}]
 dyn_l_back_chain_03 [{shapeType=Capsule, length=2.5, radius=2, rotation=<rot 90 0 0>}]
 dyn_l_back_chain_04 [{shapeType=Capsule, length=2.5, radius=2, rotation=<rot 90 0 0>}]
 dyn_l_back_chain_05 [{shapeType=Capsule, length=3.5, radius=2, center=<2.9117 0 0.0002>, rotation=<rot 90 0 0>}]
 dyn_l_back_chain_master [{shapeType=Capsule, length=0.5, radius=0.5}]
 dyn_l_front_chain_01 [{shapeType=Capsule, length=2.5, radius=2, rotation=<rot 90 0 0>}]
 dyn_l_front_chain_02 [{shapeType=Capsule, length=2.5, radius=2, rotation=<rot 90 0 0>}]
 dyn_l_front_chain_03 [{shapeType=Capsule, length=2.5, radius=2, rotation=<rot 90 0 0>}]
 dyn_l_front_chain_04 [{shapeType=Capsule, length=2.5, radius=2, rotation=<rot 90 0 0>}]
 dyn_l_front_chain_05 [{shapeType=Capsule, length=3.5, radius=2, rotation=<rot 90 0 0>}]
 dyn_l_front_chain_master [{shapeType=Capsule, length=0.5, radius=0.5}]
 dyn_r_back_chain_01 [{shapeType=Capsule, length=2.5, radius=1.75, center=<2.4604 0.0001 -0.0216>, rotation=<rot -89.4387 -179.9985 179.9986>}]
 dyn_r_back_chain_02 [{shapeType=Capsule, length=2.3875, radius=2.184, center=<3.0907 0 0.0307>, rotation=<rot -89.4314 0.0016 -0.0018>}]
 dyn_r_back_chain_03 [{shapeType=Capsule, length=3.9662, radius=2.0593, center=<3.796 0.0001 -0.1209>, rotation=<rot -88.1626 -179.999 179.9995>}]
 dyn_r_back_chain_04 [{shapeType=Capsule, length=1, radius=2, center=<1.8694 0 0.0506>, rotation=<rot 87.9076 179.997 179.9969>}]
 dyn_r_back_chain_05 [{shapeType=Capsule, length=1.75, radius=2.25, center=<3.1274 0 0>, rotation=<rot 90 0 0>}]
 dyn_r_back_chain_master [{shapeType=Capsule, length=0.5, radius=0.5}]
 dyn_r_front_chain_01 [{shapeType=Capsule, length=3, radius=1.5, center=<3.0447 0.0001 -0.7946>, rotation=<rot -83.1025 0.0003 -0.0002>}]
 dyn_r_front_chain_02 [{shapeType=Capsule, length=1.25, radius=1.75, center=<2.3602 -0.0001 -0.0233>, rotation=<rot -89.3896 179.9983 -179.9985>}]
 dyn_r_front_chain_03 [{shapeType=Capsule, length=1, radius=2, center=<2.4017 0.0001 -0.0848>, rotation=<rot -87.9711 179.9999 -179.9999>}]
 dyn_r_front_chain_04 [{shapeType=Capsule, length=1, radius=2, center=<1.6172 0 -0.0092>, rotation=<rot -89.5748 -179.9995 179.9994>}]
 dyn_r_front_chain_05 [{shapeType=Capsule, length=1.75, radius=2.25, center=<3.0446 0.0001 0>, rotation=<rot 90 0 0>}]
 dyn_r_front_chain_master [{shapeType=Capsule, length=0.5, radius=1.25, center=<0 0.0006 -1.9611>, rotation=<rot 0 0 60.1489>}]
 pelvis [{shapeType=Capsule, length=13.2429, radius=7.7054}]
 thigh_l [
  {shapeType=Capsule, length=17.2433, radius=7.8679, center=<-15.493 0 0>, rotation=<rot 90 0 0>}
  {shapeType=Capsule, length=17.9933, radius=8.1179, center=<-15.5659 0 5.7571>, rotation=<rot 90 0 0>}
 ]
 thigh_r [
  {shapeType=Capsule, length=14.7081, radius=5.1583, center=<15.3328 0.0002 -0.0868>, rotation=<rot 89.511 -180 179.9999>}
  {shapeType=Capsule, length=18.3973, radius=7.7139, center=<15.1813 -0.2433 -0.1862>, rotation=<rot 86.474 0 0>}
  {shapeType=Capsule, length=13.4725, radius=5.5016, center=<15.2792 0 -6.3631>, rotation=<rot 90 0 0>}
 ]
}
constraints [
 {bone1Name=thigh_l, bone2Name=pelvis, swing1LimitDegrees=45, swing2LimitDegrees=45, twistLimitDegrees=45}
 {bone1Name=thigh_r, bone2Name=pelvis, swing1LimitDegrees=45, swing2LimitDegrees=45, twistLimitDegrees=45}
 {bone1Name=dyn_r_back_chain_01, bone2Name=dyn_r_back_chain_master, swing1LimitDegrees=30, swing2LimitDegrees=5, twistLimitDegrees=0.5}
 {bone1Name=dyn_r_back_chain_02, bone2Name=dyn_r_back_chain_01, swing1LimitDegrees=25, swing2LimitDegrees=25, twistLimitDegrees=0.5}
 {bone1Name=dyn_r_back_chain_03, bone2Name=dyn_r_back_chain_02, swing1LimitDegrees=20, swing2LimitDegrees=25, twistLimitDegrees=0.5}
 {bone1Name=dyn_r_back_chain_04, bone2Name=dyn_r_back_chain_03, swing1LimitDegrees=20, swing2LimitDegrees=20, twistLimitDegrees=0.2}
 {bone1Name=dyn_r_back_chain_05, bone2Name=dyn_r_back_chain_04, swing1LimitDegrees=35, swing2LimitDegrees=25, twistLimitDegrees=0.2}
 {bone1Name=dyn_r_front_chain_01, bone2Name=dyn_r_front_chain_master, swing1LimitDegrees=35, swing2LimitDegrees=5, twistLimitDegrees=0.5}
 {bone1Name=dyn_r_front_chain_02, bone2Name=dyn_r_front_chain_01, swing1LimitDegrees=35, swing2LimitDegrees=25, twistLimitDegrees=0.5}
 {bone1Name=dyn_r_front_chain_03, bone2Name=dyn_r_front_chain_02, swing1LimitDegrees=35, swing2LimitDegrees=25, twistLimitDegrees=0.5}
 {bone1Name=dyn_r_front_chain_04, bone2Name=dyn_r_front_chain_03, swing1LimitDegrees=35, swing2LimitDegrees=25, twistLimitDegrees=0.5}
 {bone1Name=dyn_r_front_chain_05, bone2Name=dyn_r_front_chain_04, swing1LimitDegrees=35, swing2LimitDegrees=25, twistLimitDegrees=0.5}
 {bone1Name=dyn_l_front_chain_01, bone2Name=dyn_l_front_chain_master, swing1LimitDegrees=35, swing2LimitDegrees=5, twistLimitDegrees=5}
 {bone1Name=dyn_l_front_chain_02, bone2Name=dyn_l_front_chain_01, swing1LimitDegrees=45, swing2LimitDegrees=45, twistLimitDegrees=20}
 {bone1Name=dyn_l_front_chain_03, bone2Name=dyn_l_front_chain_02, swing1LimitDegrees=35, swing2LimitDegrees=25, twistLimitDegrees=10}
 {bone1Name=dyn_l_front_chain_04, bone2Name=dyn_l_front_chain_03, swing1LimitDegrees=35, swing2LimitDegrees=25, twistLimitDegrees=10}
 {bone1Name=dyn_l_front_chain_05, bone2Name=dyn_l_front_chain_04, swing1LimitDegrees=12, swing2LimitDegrees=25, twistLimitDegrees=10}
 {bone1Name=dyn_l_back_chain_01, bone2Name=dyn_l_back_chain_master, swing1LimitDegrees=15, swing2LimitDegrees=5, twistLimitDegrees=5}
 {bone1Name=dyn_l_back_chain_02, bone2Name=dyn_l_back_chain_01, swing1LimitDegrees=15, swing2LimitDegrees=25, twistLimitDegrees=10}
 {bone1Name=dyn_l_back_chain_03, bone2Name=dyn_l_back_chain_02, swing1LimitDegrees=15, swing2LimitDegrees=25, twistLimitDegrees=10}
 {bone1Name=dyn_l_back_chain_04, bone2Name=dyn_l_back_chain_03, swing1LimitDegrees=15, swing2LimitDegrees=25, twistLimitDegrees=10}
 {bone1Name=dyn_l_back_chain_05, bone2Name=dyn_l_back_chain_04, swing1LimitDegrees=12, swing2LimitDegrees=25, twistLimitDegrees=10}
 {
  bone1Name=dyn_l_back_chain_master
  bone2Name=pelvis
  swing1LimitDegrees=45
  swing1Motion=Locked
  swing2LimitDegrees=45
  swing2Motion=Locked
  twistLimitDegrees=45
  twistMotion=Locked
 }
 {
  bone1Name=dyn_l_front_chain_master
  bone2Name=pelvis
  swing1LimitDegrees=45
  swing1Motion=Locked
  swing2LimitDegrees=45
  swing2Motion=Locked
  twistLimitDegrees=45
  twistMotion=Locked
 }
 {
  bone1Name=dyn_r_front_chain_master
  bone2Name=pelvis
  swing1LimitDegrees=45
  swing1Motion=Locked
  swing2LimitDegrees=45
  swing2Motion=Locked
  twistLimitDegrees=45
  twistMotion=Locked
 }
 {
  bone1Name=dyn_r_back_chain_master
  bone2Name=pelvis
  swing1LimitDegrees=45
  swing1Motion=Locked
  swing2LimitDegrees=45
  swing2Motion=Locked
  twistLimitDegrees=45
  twistMotion=Locked
 }
 {bone1Name=dyn_l_front_chain_05, bone2Name=dyn_l_back_chain_05, swing1LimitDegrees=1, swing2LimitDegrees=1, twistLimitDegrees=45}
 {bone1Name=dyn_l_front_chain_04, bone2Name=dyn_l_back_chain_04, swing1LimitDegrees=1, swing2LimitDegrees=1, twistLimitDegrees=45}
 {bone1Name=dyn_l_back_chain_04, bone2Name=dyn_l_front_chain_04, swing1LimitDegrees=1, swing2LimitDegrees=1, twistLimitDegrees=45}
 {bone1Name=dyn_r_back_chain_04, bone2Name=dyn_r_front_chain_04, swing1LimitDegrees=1, swing2LimitDegrees=1, twistLimitDegrees=0.5}
 {bone1Name=dyn_r_front_chain_04, bone2Name=dyn_r_back_chain_04, swing1LimitDegrees=1, swing2LimitDegrees=1, twistLimitDegrees=0.5}
 {bone1Name=dyn_r_front_chain_05, bone2Name=dyn_r_back_chain_05, swing1LimitDegrees=1, swing2LimitDegrees=1, twistLimitDegrees=0.5}
 {bone1Name=dyn_front_flap, bone2Name=pelvis, swing1LimitDegrees=8, swing2LimitDegrees=60, twistLimitDegrees=5}
 {bone1Name=dyn_back_Flap, bone2Name=dyn_back_Flap_master, swing1LimitDegrees=8, swing2LimitDegrees=30, twistLimitDegrees=3}
 {
  bone1Name=dyn_back_Flap_master
  bone2Name=pelvis
  swing1LimitDegrees=45
  swing1Motion=Free
  swing2LimitDegrees=45
  swing2Motion=Free
  twistLimitDegrees=45
  twistMotion=Free
 }
 {bone1Name=dyn_front_flap_r_01, bone2Name=dyn_front_flap, swing1LimitDegrees=30, swing2LimitDegrees=30, twistLimitDegrees=10}
 {bone1Name=dyn_front_flap_r_02, bone2Name=dyn_front_flap_r_01, swing1LimitDegrees=30, swing2LimitDegrees=30, twistLimitDegrees=10}
 {bone1Name=dyn_front_flap_l_01, bone2Name=dyn_front_flap, swing1LimitDegrees=30, swing2LimitDegrees=30, twistLimitDegrees=10}
 {bone1Name=dyn_front_flap_l_02, bone2Name=dyn_front_flap_l_01, swing1LimitDegrees=30, swing2LimitDegrees=30, twistLimitDegrees=10}
]
