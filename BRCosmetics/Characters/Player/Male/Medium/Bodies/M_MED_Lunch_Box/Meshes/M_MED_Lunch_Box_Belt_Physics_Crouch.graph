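# Physics asset for the Lunch Box male medium body's belt, named as the crouch variant, covering dangling chains and flaps around pelvis and thigh shapes.
# Four five-segment capsule chains run from masters with free angular motion on the pelvis.
# Tight-swing constraints tie front and back chains together at their last two segments on each side, forming loops.
# The back flap hangs from the pelvis; the front flap uses a boxed master and branches left and right.
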PhysicsAsset /BRCosmetics/Characters/Player/Male/Medium/Bodies/M_MED_Lunch_Box/Meshes/M_MED_Lunch_Box_Belt_Physics_Crouch
bodies {
 dyn_back_Flap [{shapeType=Capsule, length=6, radius=2, center=<2.9548 0 -0.7427>, rotation=<rot 0 0 90>}]
 dyn_front_flap [{shapeType=Capsule, length=4.25, radius=2.75, rotation=<rot 0 0 90>}]
 dyn_front_flap_l_01 [{shapeType=Capsule, length=4.6664, radius=2.8356, center=<3.7055 0 0>, rotation=<rot 90 0 0>}]
 dyn_front_flap_l_02 [{shapeType=Capsule, length=4.25, radius=2.75, center=<3.1284 0 -0.0001>, rotation=<rot 90 0 0>}]
 dyn_front_flap_master [{shapeType=Box, extentX=8.6157, extentY=13.9336, extentZ=0.5366, center=<4.6301 0.0001 -1.8518>}]
 dyn_front_flap_r_01 [{shapeType=Capsule, length=2, radius=2.75, center=<3.1284 0 -0.0001>, rotation=<rot 90 0 0>}]
 dyn_front_flap_r_02 [{shapeType=Capsule, length=2, radius=2.75, center=<3.1284 0 -0.0001>, rotation=<rot 90 0 0>}]
 dyn_l_back_chain_01 [{shapeType=Capsule, length=2, radius=2.5, center=<2.0301 -0.0001 0.0662>, rotation=<rot -88.1332 0.0001 -0.0001>}]
 dyn_l_back_chain_02 [{shapeType=Capsule, length=2, radius=2.5, center=<2.0301 -0.0001 0.0662>, rotation=<rot -88.1318 -0.0016 0.0016>}]
 dyn_l_back_chain_03 [{shapeType=Capsule, length=2, radius=2.5, center=<2.0302 -0.0001 0.0639>, rotation=<rot -88.1972 -0.0004 0.0004>}]
 dyn_l_back_chain_04 [{shapeType=Capsule, length=2, radius=2.5, center=<2.0302 -0.0001 0.0648>, rotation=<rot -88.1719 0.0028 -0.0028>}]
 dyn_l_back_chain_05 [{shapeType=Capsule, length=1.7502, radius=2.2498, center=<3.0913 0 0.0002>, rotation=<rot 90 0 0>}]
 dyn_l_back_chain_master [{shapeType=Capsule, length=0.5, radius=0.5}]
 dyn_l_front_chain_01 [{shapeType=Capsule, length=2, radius=2, center=<2.6993 0 0.7412>, rotation=<rot -78.6776 -0.0001 0>}]
 dyn_l_front_chain_02 [{shapeType=Capsule, length=3, radius=1.5, center=<2.03 -0.0001 0.0707>, rotation=<rot -88.0044 -0.0014 0.0015>}]
 dyn_l_front_chain_03 [{shapeType=Capsule, length=3.5, radius=1, center=<2.0302 -0.0001 0.0627>, rotation=<rot -88.2299 -0.0035 0.0036>}]
 dyn_l_front_chain_04 [{shapeType=Capsule, length=2, radius=2.5, center=<2.0301 -0.0001 0.0685>, rotation=<rot -88.0674 -0.0031 0.0031>}]
 dyn_l_front_chain_05 [{shapeType=Capsule, length=1.7502, radius=2.2498, center=<3.0244 0 0.0001>, rotation=<rot 90 0 0>}]
 dyn_l_front_chain_master [{shapeType=Capsule, length=0.5, radius=0.5}]
 dyn_r_back_chain_01 [{shapeType=Capsule, length=2, radius=2.25, center=<1.46 0.0001 0>, rotation=<rot -90 0.0395 -0.0395>}]
 dyn_r_back_chain_02 [{shapeType=Capsule, length=2, radius=2.25, center=<1.46 0.0001 0>, rotation=<rot -90 0.0395 -0.0395>}]
 dyn_r_back_chain_03 [{shapeType=Capsule, length=2, radius=2.25, center=<1.46 0.0001 0>, rotation=<rot -90 0.0395 -0.0395>}]
 dyn_r_back_chain_04 [{shapeType=Capsule, length=2, radius=2.25, center=<1.46 0.0001 0>, rotation=<rot -90 0.0033 -0.0032>}]
 dyn_r_back_chain_05 [{shapeType=Capsule, length=1.9709, radius=2.0291, center=<2.9506 0 0>, rotation=<rot 90 0 0>}]
 dyn_r_back_chain_master [{shapeType=Capsule, length=0.5, radius=0.5}]
 dyn_r_front_chain_01 [{shapeType=Capsule, length=2, radius=2.25, center=<1.46 0.0001 0>, rotation=<rot -90 0.0131 -0.0128>}]
 dyn_r_front_chain_02 [{shapeType=Capsule, length=2, radius=2.25, center=<1.46 0.0001 0>, rotation=<rot -90 -0.0007 0.0009>}]
 dyn_r_front_chain_03 [{shapeType=Capsule, length=2, radius=2.25, center=<1.46 0.0001 0>, rotation=<rot -90 0.0197 -0.0197>}]
 dyn_r_front_chain_04 [{shapeType=Capsule, length=2, radius=2.25, center=<1.46 0.0001 0.0036>, rotation=<rot -89.8573 -0.0061 0.0062>}]
 dyn_r_front_chain_05 [{shapeType=Capsule, length=1.9709, radius=2.0291, center=<2.9803 0 -0.0002>, rotation=<rot 90 0 0>}]
 dyn_r_front_chain_master [{shapeType=Capsule, length=0.5, radius=0.5}]
 pelvis [{shapeType=Capsule, length=13.2429, radius=7.7054}]
 thigh_l [
  {shapeType=Capsule, length=18.1419, radius=7.9694, center=<-15.493 0.4192 0>, rotation=<rot 90 -12.3391 -14.5605>}
  {shapeType=Capsule, length=17.8672, radius=5.7647, center=<-15.4897 0.0001 0.1791>, rotation=<rot 90 0 0>}
 ]
 thigh_r [{shapeType=Capsule, length=17.9027, radius=8.2085, center=<15.3412 0.0002 0>, rotation=<rot 90 0 0>}]
}
constraints [
 {bone1Name=thigh_l, bone2Name=pelvis, swing1LimitDegrees=45, swing2LimitDegrees=45, twistLimitDegrees=45}
 {bone1Name=thigh_r, bone2Name=pelvis, swing1LimitDegrees=45, swing2LimitDegrees=45, twistLimitDegrees=45}
 {
  bone1Name=dyn_r_back_chain_01
  bone2Name=dyn_r_back_chain_master
  swing1LimitDegrees=30
  swing1Motion=Locked
  swing2LimitDegrees=2
  swing2Motion=Locked
  twistLimitDegrees=2
  twistMotion=Locked
 }
 {bone1Name=dyn_r_back_chain_02, bone2Name=dyn_r_back_chain_01, swing1LimitDegrees=35, swing2LimitDegrees=25, twistLimitDegrees=10}
 {bone1Name=dyn_r_back_chain_03, bone2Name=dyn_r_back_chain_02, swing1LimitDegrees=35, swing2LimitDegrees=25, twistLimitDegrees=10}
 {bone1Name=dyn_r_back_chain_04, bone2Name=dyn_r_back_chain_03, swing1LimitDegrees=35, swing2LimitDegrees=25, twistLimitDegrees=10}
 {bone1Name=dyn_r_back_chain_05, bone2Name=dyn_r_back_chain_04, swing1LimitDegrees=35, swing2LimitDegrees=25, twistLimitDegrees=10}
 {bone1Name=dyn_r_front_chain_01, bone2Name=dyn_r_front_chain_master, swing1LimitDegrees=15, swing2LimitDegrees=5, twistLimitDegrees=5}
 {bone1Name=dyn_r_front_chain_02, bone2Name=dyn_r_front_chain_01, swing1LimitDegrees=35, swing2LimitDegrees=25, twistLimitDegrees=10}
 {bone1Name=dyn_r_front_chain_03, bone2Name=dyn_r_front_chain_02, swing1LimitDegrees=35, swing2LimitDegrees=25, twistLimitDegrees=10}
 {bone1Name=dyn_r_front_chain_04, bone2Name=dyn_r_front_chain_03, swing1LimitDegrees=35, swing2LimitDegrees=25, twistLimitDegrees=10}
 {bone1Name=dyn_r_front_chain_05, bone2Name=dyn_r_front_chain_04, swing1LimitDegrees=35, swing2LimitDegrees=25, twistLimitDegrees=10}
 {bone1Name=dyn_l_front_chain_01, bone2Name=dyn_l_front_chain_master, swing1LimitDegrees=35, swing2LimitDegrees=2, twistLimitDegrees=3.5}
 {bone1Name=dyn_l_front_chain_02, bone2Name=dyn_l_front_chain_01, swing1LimitDegrees=15, swing2LimitDegrees=15, twistLimitDegrees=20}
 {bone1Name=dyn_l_front_chain_03, bone2Name=dyn_l_front_chain_02, swing1LimitDegrees=35, swing2LimitDegrees=25, twistLimitDegrees=10}
 {bone1Name=dyn_l_front_chain_04, bone2Name=dyn_l_front_chain_03, swing1LimitDegrees=35, swing2LimitDegrees=25, twistLimitDegrees=10}
 {bone1Name=dyn_l_front_chain_05, bone2Name=dyn_l_front_chain_04, swing1LimitDegrees=35, swing2LimitDegrees=25, twistLimitDegrees=10}
 {bone1Name=dyn_l_back_chain_01, bone2Name=dyn_l_back_chain_master, swing1LimitDegrees=35, swing2LimitDegrees=5, twistLimitDegrees=5}
 {bone1Name=dyn_l_back_chain_02, bone2Name=dyn_l_back_chain_01, swing1LimitDegrees=35, swing2LimitDegrees=25, twistLimitDegrees=10}
 {bone1Name=dyn_l_back_chain_03, bone2Name=dyn_l_back_chain_02, swing1LimitDegrees=35, swing2LimitDegrees=25, twistLimitDegrees=10}
 {bone1Name=dyn_l_back_chain_04, bone2Name=dyn_l_back_chain_03, swing1LimitDegrees=35, swing2LimitDegrees=25, twistLimitDegrees=10}
 {bone1Name=dyn_l_back_chain_05, bone2Name=dyn_l_back_chain_04, swing1LimitDegrees=35, swing2LimitDegrees=25, twistLimitDegrees=10}
 {
  bone1Name=dyn_l_back_chain_master
  bone2Name=pelvis
  swing1LimitDegrees=45
  swing1Motion=Free
  swing2LimitDegrees=45
  swing2Motion=Free
  twistLimitDegrees=45
  twistMotion=Free
 }
 {
  bone1Name=dyn_l_front_chain_master
  bone2Name=pelvis
  swing1LimitDegrees=45
  swing1Motion=Free
  swing2LimitDegrees=45
  swing2Motion=Free
  twistLimitDegrees=45
  twistMotion=Free
 }
 {
  bone1Name=dyn_r_front_chain_master
  bone2Name=pelvis
  swing1LimitDegrees=45
  swing1Motion=Free
  swing2LimitDegrees=45
  swing2Motion=Free
  twistLimitDegrees=45
  twistMotion=Free
 }
 {
  bone1Name=dyn_r_back_chain_master
  bone2Name=pelvis
  swing1LimitDegrees=45
  swing1Motion=Free
  swing2LimitDegrees=45
  swing2Motion=Free
  twistLimitDegrees=45
  twistMotion=Free
 }
 {bone1Name=dyn_l_front_chain_05, bone2Name=dyn_l_back_chain_05, swing1LimitDegrees=1, swing2LimitDegrees=1, twistLimitDegrees=45}
 {bone1Name=dyn_l_front_chain_04, bone2Name=dyn_l_back_chain_04, swing1LimitDegrees=1, swing2LimitDegrees=1, twistLimitDegrees=45}
 {bone1Name=dyn_l_back_chain_04, bone2Name=dyn_l_front_chain_04, swing1LimitDegrees=1, swing2LimitDegrees=1, twistLimitDegrees=45}
 {bone1Name=dyn_r_back_chain_04, bone2Name=dyn_r_front_chain_04, swing1LimitDegrees=1, swing2LimitDegrees=1, twistLimitDegrees=45}
 {bone1Name=dyn_r_front_chain_04, bone2Name=dyn_r_back_chain_04, swing1LimitDegrees=1, swing2LimitDegrees=1, twistLimitDegrees=45}
 {bone1Name=dyn_r_front_chain_05, bone2Name=dyn_r_back_chain_05, swing1LimitDegrees=1, swing2LimitDegrees=1, twistLimitDegrees=45}
 {bone1Name=dyn_front_flap, bone2Name=dyn_front_flap_master, swing1LimitDegrees=8, swing2LimitDegrees=35, twistLimitDegrees=20}
 {bone1Name=dyn_back_Flap, bone2Name=pelvis, swing1LimitDegrees=8, swing2LimitDegrees=30, twistLimitDegrees=12}
 {bone1Name=dyn_front_flap_master, bone2Name=pelvis, swing1LimitDegrees=20, swing2LimitDegrees=20, twistLimitDegrees=15}
 {bone1Name=dyn_front_flap_r_01, bone2Name=dyn_front_flap, swing1LimitDegrees=30, swing2LimitDegrees=30, twistLimitDegrees=10}
 {bone1Name=dyn_front_flap_r_02, bone2Name=dyn_front_flap_r_01, swing1LimitDegrees=30, swing2LimitDegrees=30, twistLimitDegrees=10}
 {
  bone1Name=dyn_front_flap_l_01
  bone2Name=dyn_front_flap
  swing1LimitDegrees=45
  swing1Motion=Locked
  swing2LimitDegrees=30
  twistLimitDegrees=10
 }
 {bone1Name=dyn_front_flap_l_02, bone2Name=dyn_front_flap_l_01, swing1LimitDegrees=30, swing2LimitDegrees=30, twistLimitDegrees=10}
]
