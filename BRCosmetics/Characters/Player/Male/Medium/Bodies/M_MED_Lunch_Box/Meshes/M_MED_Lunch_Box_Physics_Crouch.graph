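# Physics asset for the Lunch Box medium male player body, named as its crouch variant, covering the jacket's dynamic bones.
# Seven two-bone capsule chains ring the jacket at front, back and sides, each constrained to the pelvis with small angular limits.
# Front-left limits are tuned per axis, while the other chains share uniform values.
# Larger pelvis, spine and thigh capsules form the body volume, joined by free angular constraints.
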PhysicsAsset /BRCosmetics/Characters/Player/Male/Medium/Bodies/M_MED_Lunch_Box/Meshes/M_MED_Lunch_Box_Physics_Crouch
bodies {
 dyn_jacket_bk_01 [{shapeType=Capsule, length=5.25, radius=2.25, center=<2.1984 0 0.4223>, rotation=<rot 0 0 90>}]
 dyn_jacket_bk_02 [{shapeType=Capsule, length=3.75, radius=2.25, center=<2.3276 0 -1.4127>, rotation=<rot 0 0 90>}]
 dyn_jacket_bk_l_01 [{shapeType=Capsule, length=6, radius=2, center=<1.2544 0 0>, rotation=<rot 0 0 90>}]
 dyn_jacket_bk_l_02 [{shapeType=Capsule, length=5.5, radius=2.25, center=<2.2124 -0.0001 -1.69>, rotation=<rot 0 0 90>}]
 dyn_jacket_bk_r_01 [{shapeType=Capsule, length=3.75, radius=2.25, center=<1.7687 0 0.0002>, rotation=<rot 0 0 90>}]
 dyn_jacket_bk_r_02 [{shapeType=Capsule, length=4.25, radius=2.25, center=<2.5493 0 -2.9719>, rotation=<rot 0 0 90>}]
 dyn_jacket_fr_l_01 [{shapeType=Capsule, length=4.25, radius=3.25, center=<2.9456 0 0>, rotation=<rot 0 0 90>}]
 dyn_jacket_fr_l_02 [{shapeType=Capsule, length=2.5, radius=2, center=<3.9836 0.5355 -0.8897>, rotation=<rot 0 0 96.2725>}]
 dyn_jacket_fr_r_01 [{shapeType=Capsule, length=4.5, radius=2, center=<1.6581 0 -0.0002>, rotation=<rot 0 0 90>}]
 dyn_jacket_fr_r_02 [{shapeType=Capsule, length=5.75, radius=1.75, center=<8.4321 0 0>, rotation=<rot 0 0 90>}]
 dyn_jacket_l_01 [{shapeType=Capsule, length=6.5, radius=1.75, center=<3.083 0 0.0001>, rotation=<rot 0 0 90>}]
 dyn_jacket_l_02 [{shapeType=Capsule, length=6.25, radius=2.25, center=<2.1852 0 0>, rotation=<rot 0 0 90>}]
 dyn_jacket_r_01 [{shapeType=Capsule, length=6, radius=1.75, center=<1.7328 0 0>, rotation=<rot 0 0 90>}]
 dyn_jacket_r_02 [{shapeType=Capsule, length=5.25, radius=2, center=<2.5369 0 0.0001>, rotation=<rot 0 0 90>}]
 pelvis [{shapeType=Capsule, length=9.1701, radius=9.4934}]
 spine_01 [{shapeType=Capsule, length=8.5834, radius=8.7168, center=<0 0.7969 0>}]
 spine_02 [{shapeType=Capsule, length=8.5123, radius=7.7944}]
 spine_03 [{shapeType=Capsule, length=7.86, radius=8.1}]
 thigh_l [{shapeType=Capsule, length=24.8769, radius=7.8394, center=<-19.9743 0.1731 0>, rotation=<rot 90 0 -3.9292>}]
 thigh_r [{shapeType=Capsule, length=22.888, radius=7.821, center=<18.5981 -0.0003 0>, rotation=<rot 90 0 0>}]
}
constraints [
 {
  bone1Name=spine_01
  bone2Name=pelvis
  swing1LimitDegrees=45
  swing1Motion=Free
  swing2LimitDegrees=45
  swing2Motion=Free
  twistLimitDegrees=45
  twistMotion=Free
 }
 {
  bone1Name=spine_02
  bone2Name=spine_01
  swing1LimitDegrees=45
  swing1Motion=Free
  swing2LimitDegrees=45
  swing2Motion=Free
  twistLimitDegrees=45
  twistMotion=Free
 }
 {
  bone1Name=spine_03
  bone2Name=spine_02
  swing1LimitDegrees=45
  swing1Motion=Free
  swing2LimitDegrees=45
  swing2Motion=Free
  twistLimitDegrees=45
  twistMotion=Free
 }
 {bone1Name=dyn_jacket_l_01, bone2Name=pelvis, swing1LimitDegrees=15, swing2LimitDegrees=15, twistLimitDegrees=15}
 {bone1Name=dyn_jacket_l_02, bone2Name=dyn_jacket_l_01, swing1LimitDegrees=15, swing2LimitDegrees=15, twistLimitDegrees=15}
 {bone1Name=dyn_jacket_fr_l_01, bone2Name=pelvis, swing1LimitDegrees=10, swing2LimitDegrees=20, twistLimitDegrees=5}
 {bone1Name=dyn_jacket_fr_l_02, bone2Name=dyn_jacket_fr_l_01, swing1LimitDegrees=20, swing2LimitDegrees=28, twistLimitDegrees=10}
 {bone1Name=dyn_jacket_bk_l_01, bone2Name=pelvis, swing1LimitDegrees=15, swing2LimitDegrees=15, twistLimitDegrees=15}
 {bone1Name=dyn_jacket_bk_l_02, bone2Name=dyn_jacket_bk_l_01, swing1LimitDegrees=15, swing2LimitDegrees=15, twistLimitDegrees=15}
 {bone1Name=dyn_jacket_bk_01, bone2Name=pelvis, swing1LimitDegrees=15, swing2LimitDegrees=15, twistLimitDegrees=15}
 {bone1Name=dyn_jacket_bk_02, bone2Name=dyn_jacket_bk_01, swing1LimitDegrees=15, swing2LimitDegrees=15, twistLimitDegrees=15}
 {bone1Name=dyn_jacket_fr_r_01, bone2Name=pelvis, swing1LimitDegrees=15, swing2LimitDegrees=15, twistLimitDegrees=15}
 {bone1Name=dyn_jacket_fr_r_02, bone2Name=dyn_jacket_fr_r_01, swing1LimitDegrees=15, swing2LimitDegrees=15, twistLimitDegrees=15}
 {bone1Name=dyn_jacket_bk_r_01, bone2Name=pelvis, swing1LimitDegrees=15, swing2LimitDegrees=15, twistLimitDegrees=15}
 {bone1Name=dyn_jacket_bk_r_02, bone2Name=dyn_jacket_bk_r_01, swing1LimitDegrees=15, swing2LimitDegrees=15, twistLimitDegrees=15}
 {bone1Name=dyn_jacket_r_01, bone2Name=pelvis, swing1LimitDegrees=15, swing2LimitDegrees=15, twistLimitDegrees=15}
 {bone1Name=dyn_jacket_r_02, bone2Name=dyn_jacket_r_01, swing1LimitDegrees=15, swing2LimitDegrees=15, twistLimitDegrees=15}
 {
  bone1Name=thigh_l
  bone2Name=pelvis
  swing1LimitDegrees=45
  swing1Motion=Free
  swing2LimitDegrees=45
  swing2Motion=Free
  twistLimitDegrees=45
  twistMotion=Free
 }
 {
  bone1Name=thigh_r
  bone2Name=pelvis
  swing1LimitDegrees=45
  swing1Motion=Free
  swing2LimitDegrees=45
  swing2Motion=Free
  twistLimitDegrees=45
  twistMotion=Free
 }
]
