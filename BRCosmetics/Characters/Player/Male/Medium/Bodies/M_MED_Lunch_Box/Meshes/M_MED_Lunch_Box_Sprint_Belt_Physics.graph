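# Physics asset for the Sprint belt on the male medium Lunch Box body cosmetic, covering dangling chains and front and back flaps.
# Four five-segment capsule chains hang from master bodies with free angular joints to the pelvis, beside pelvis and thigh volumes.
# On each side, front and back chains are cross-linked at their last two segments with tight swing limits.
# Flaps use limited joints to their own masters, which have no pelvis constraint.
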PhysicsAsset /BRCosmetics/Characters/Player/Male/Medium/Bodies/M_MED_Lunch_Box/Meshes/M_MED_Lunch_Box_Sprint_Belt_Physics
bodies {
 dyn_back_Flap [
  {shapeType=Capsule, length=6, radius=4, center=<4.0706 0.0001 -0.0483>, rotation=<rot -89.1215 179.997 -91.5986>}
  {
   shapeType=Capsule
   length=16.5599
   radius=1.8992
   shapeName=dyn_back_Flap_capsule
   center=<-1.5581 0.0084 -1.5887>
   rotation=<rot -0.0002 0 -90.3012>
  }
 ]
 dyn_back_Flap_master [{shapeType=Capsule, length=1, radius=1}]
 dyn_front_flap [{shapeType=Capsule, length=17.0123, radius=2.6242, center=<2.0899 0.0089 -0.1499>, rotation=<rot -87.3983 0 -89.7555>}]
 dyn_front_flap01 [{shapeType=Capsule, length=14.4495, radius=2.7937, shapeName=dyn_front_flap01_capsule, center=<3.3646 0 0>, rotation=<rot 0.0001 0 -90>}]
 dyn_front_flap02 [{
   shapeType=Capsule
   length=11.8819
   radius=2.0384
   shapeName=dyn_front_flap02_capsule
   center=<2.0642 0.1694 -0.2302>
   rotation=<rot 0 -178.221 -88.7859>
  }]
 dyn_front_flap_master [{shapeType=Sphere, radius=0.505, shapeName=dyn_front_flap_master_sphere}]
 dyn_l_back_chain_01 [{shapeType=Capsule, length=2.75, radius=1.75, center=<2.5048 0 -0.0318>, rotation=<rot -89.2285 179.9964 -179.9964>}]
 dyn_l_back_chain_02 [{shapeType=Capsule, length=1.5, radius=2, center=<2.0331 0.0001 -0.0034>, rotation=<rot -89.7851 179.9913 -179.9912>}]
 dyn_l_back_chain_03 [{shapeType=Capsule, length=2.75, radius=2, center=<2.7497 0.0002 -0.0286>, rotation=<rot -89.3629 179.9915 -179.9915>}]
 dyn_l_back_chain_04 [{shapeType=Capsule, length=1.5, radius=2, center=<2.1217 0.0003 0.0936>, rotation=<rot -87.4745 0.0023 -0.0023>}]
 dyn_l_back_chain_05 [{shapeType=Capsule, length=2, radius=2.5, center=<3.0697 0 0>, rotation=<rot 90 0 0>}]
 dyn_l_back_chain_master [{shapeType=Capsule, length=0.5, radius=0.5}]
 dyn_l_front_chain_01 [{shapeType=Capsule, length=0.5, radius=3, center=<2.3832 0 0.2152>, rotation=<rot -82.2249 -0.0003 0.0003>}]
 dyn_l_front_chain_02 [{shapeType=Capsule, length=1, radius=2.5, center=<1.776 -0.0002 0.0088>, rotation=<rot -89.7154 -0.0002 0.0002>}]
 dyn_l_front_chain_03 [{shapeType=Capsule, length=1.75, radius=2.5, center=<1.9536 0 -0.0218>, rotation=<rot -89.3741 -179.998 179.9981>}]
 dyn_l_front_chain_04 [{shapeType=Capsule, length=1.25, radius=2.25, center=<1.6122 -0.0001 0.025>, rotation=<rot -89.1111 -0.0053 0.0053>}]
 dyn_l_front_chain_05 [{shapeType=Capsule, length=2, radius=2.5, center=<3.068 0 0>, rotation=<rot 90 0 0>}]
 dyn_l_front_chain_master [{shapeType=Capsule, length=0.5, radius=0.5}]
 dyn_r_back_chain_01 [{shapeType=Capsule, length=2.75, radius=2, center=<2.4604 0.0001 -0.0216>, rotation=<rot -89.4387 -179.9985 179.9986>}]
 dyn_r_back_chain_02 [{shapeType=Capsule, length=1.5, radius=2, center=<1.8524 0 0.0184>, rotation=<rot -89.4314 0.0016 -0.0018>}]
 dyn_r_back_chain_03 [{shapeType=Capsule, length=3, radius=2, center=<3.1113 0.0001 -0.0985>, rotation=<rot -88.1694 -179.999 179.9995>}]
 dyn_r_back_chain_04 [{shapeType=Capsule, length=1.25, radius=2.25, center=<1.8694 0 0.0506>, rotation=<rot 87.9076 179.997 179.9969>}]
 dyn_r_back_chain_05 [{shapeType=Capsule, length=2, radius=2.5, center=<3.1274 0 0>, rotation=<rot 90 0 0>}]
 dyn_r_back_chain_master [{shapeType=Capsule, length=0.5, radius=0.5}]
 dyn_r_front_chain_01 [{shapeType=Capsule, length=3, radius=2, center=<2.9938 0.0001 -0.0258>, rotation=<rot -89.5098 179.9996 -179.9995>}]
 dyn_r_front_chain_02 [{shapeType=Capsule, length=1.5, radius=2, center=<2.3602 -0.0001 -0.0233>, rotation=<rot -89.3896 179.9983 -179.9985>}]
 dyn_r_front_chain_03 [{shapeType=Capsule, length=1.25, radius=2.25, center=<2.4017 0.0001 -0.0848>, rotation=<rot -87.9711 179.9999 -179.9999>}]
 dyn_r_front_chain_04 [{shapeType=Capsule, length=1.25, radius=2.25, center=<1.6172 0 -0.0092>, rotation=<rot -89.5748 -179.9995 179.9994>}]
 dyn_r_front_chain_05 [{shapeType=Capsule, length=2, radius=2.5, center=<3.0446 0.0001 0>, rotation=<rot 90 0 0>}]
 dyn_r_front_chain_master [{shapeType=Capsule, length=0.5, radius=0.5}]
 pelvis [{shapeType=Capsule, length=13.2429, radius=7.7054}]
 thigh_l [
  {shapeType=Capsule, length=17.9933, radius=8.1179, center=<-15.493 0 0>, rotation=<rot 90 0 0>}
  {shapeType=Capsule, length=18.3676, radius=7.7436, center=<-15.4897 0.0001 0.3019>, rotation=<rot 90 0 0>}
  {shapeType=Capsule, length=17.9933, radius=8.1179, center=<-15.5659 0 5.7571>, rotation=<rot 90 0 0>}
 ]
 thigh_r [
  {shapeType=Capsule, length=14.7081, radius=5.1583, center=<15.3328 0.0002 -0.0868>, rotation=<rot 89.511 -180 179.9999>}
  {shapeType=Capsule, length=18.3973, radius=7.7139, center=<15.1608 -0.2433 -0.5189>, rotation=<rot 86.474 0 0>}
  {shapeType=Capsule, length=13.4725, radius=5.5016, center=<15.2792 0 -6.3631>, rotation=<rot 90 0 0>}
 ]
}
constraints [
 {bone1Name=thigh_l, bone2Name=pelvis, swing1LimitDegrees=45, swing2LimitDegrees=45, twistLimitDegrees=45}
 {bone1Name=thigh_r, bone2Name=pelvis, swing1LimitDegrees=45, swing2LimitDegrees=45, twistLimitDegrees=45}
 {bone1Name=dyn_r_back_chain_01, bone2Name=dyn_r_back_chain_master, swing1LimitDegrees=30, swing2LimitDegrees=5, twistLimitDegrees=5}
 {bone1Name=dyn_r_back_chain_02, bone2Name=dyn_r_back_chain_01, swing1LimitDegrees=35, swing2LimitDegrees=25, twistLimitDegrees=10}
 {bone1Name=dyn_r_back_chain_03, bone2Name=dyn_r_back_chain_02, swing1LimitDegrees=35, swing2LimitDegrees=25, twistLimitDegrees=10}
 {bone1Name=dyn_r_back_chain_04, bone2Name=dyn_r_back_chain_03, swing1LimitDegrees=20, swing2LimitDegrees=20, twistLimitDegrees=5}
 {bone1Name=dyn_r_back_chain_05, bone2Name=dyn_r_back_chain_04, swing1LimitDegrees=35, swing2LimitDegrees=25, twistLimitDegrees=10}
 {bone1Name=dyn_r_front_chain_01, bone2Name=dyn_r_front_chain_master, swing1LimitDegrees=35, swing2LimitDegrees=5, twistLimitDegrees=5}
 {bone1Name=dyn_r_front_chain_02, bone2Name=dyn_r_front_chain_01, swing1LimitDegrees=35, swing2LimitDegrees=25, twistLimitDegrees=10}
 {bone1Name=dyn_r_front_chain_03, bone2Name=dyn_r_front_chain_02, swing1LimitDegrees=35, swing2LimitDegrees=25, twistLimitDegrees=10}
 {bone1Name=dyn_r_front_chain_04, bone2Name=dyn_r_front_chain_03, swing1LimitDegrees=35, swing2LimitDegrees=25, twistLimitDegrees=10}
 {bone1Name=dyn_r_front_chain_05, bone2Name=dyn_r_front_chain_04, swing1LimitDegrees=35, swing2LimitDegrees=25, twistLimitDegrees=10}
 {bone1Name=dyn_l_front_chain_01, bone2Name=dyn_l_front_chain_master, swing1LimitDegrees=35, swing2LimitDegrees=5, twistLimitDegrees=5}
 {bone1Name=dyn_l_front_chain_02, bone2Name=dyn_l_front_chain_01, swing1LimitDegrees=45, swing2LimitDegrees=45, twistLimitDegrees=20}
 {bone1Name=dyn_l_front_chain_03, bone2Name=dyn_l_front_chain_02, swing1LimitDegrees=35, swing2LimitDegrees=25, twistLimitDegrees=10}
 {bone1Name=dyn_l_front_chain_04, bone2Name=dyn_l_front_chain_03, swing1LimitDegrees=35, swing2LimitDegrees=25, twistLimitDegrees=10}
 {bone1Name=dyn_l_front_chain_05, bone2Name=dyn_l_front_chain_04, swing1LimitDegrees=35, swing2LimitDegrees=25, twistLimitDegrees=10}
 {bone1Name=dyn_l_back_chain_01, bone2Name=dyn_l_back_chain_master, swing1LimitDegrees=35, swing2LimitDegrees=5, twistLimitDegrees=5}
 {bone1Name=dyn_l_back_chain_02, bone2Name=dyn_l_back_chain_01, swing1LimitDegrees=35, swing2LimitDegrees=25, twistLimitDegrees=10}
 {bone1Name=dyn_l_back_chain_03, bone2Name=dyn_l_back_chain_02, swing1LimitDegrees=35, swing2LimitDegrees=25, twistLimitDegrees=10}
 {bone1Name=dyn_l_back_chain_04, bone2Name=dyn_l_back_chain_03, swing1LimitDegrees=35, swing2LimitDegrees=25, twistLimitDegrees=10}
 {bone1Name=dyn_l_back_chain_05, bone2Name=dyn_l_back_chain_04, swing1LimitDegrees=35, swing2LimitDegrees=25, twistLimitDegrees=10}
 {
  bone1Name=dyn_l_back_chain_master
  bone2Name=pelvis
  swing1LimitDegrees=45
  swing1Motion=Free
  swing2LimitDegrees=45
  swing2Motion=Free
  twistLimitDegrees=45
  twistMotion=Free
 }
 {
  bone1Name=dyn_l_front_chain_master
  bone2Name=pelvis
  swing1LimitDegrees=45
  swing1Motion=Free
  swing2LimitDegrees=45
  swing2Motion=Free
  twistLimitDegrees=45
  twistMotion=Free
 }
 {
  bone1Name=dyn_r_front_chain_master
  bone2Name=pelvis
  swing1LimitDegrees=45
  swing1Motion=Free
  swing2LimitDegrees=45
  swing2Motion=Free
  twistLimitDegrees=45
  twistMotion=Free
 }
 {
  bone1Name=dyn_r_back_chain_master
  bone2Name=pelvis
  swing1LimitDegrees=45
  swing1Motion=Free
  swing2LimitDegrees=45
  swing2Motion=Free
  twistLimitDegrees=45
  twistMotion=Free
 }
 {bone1Name=dyn_l_front_chain_05, bone2Name=dyn_l_back_chain_05, swing1LimitDegrees=1, swing2LimitDegrees=1, twistLimitDegrees=45}
 {bone1Name=dyn_l_front_chain_04, bone2Name=dyn_l_back_chain_04, swing1LimitDegrees=1, swing2LimitDegrees=1, twistLimitDegrees=45}
 {bone1Name=dyn_l_back_chain_04, bone2Name=dyn_l_front_chain_04, swing1LimitDegrees=1, swing2LimitDegrees=1, twistLimitDegrees=45}
 {bone1Name=dyn_r_back_chain_04, bone2Name=dyn_r_front_chain_04, swing1LimitDegrees=1, swing2LimitDegrees=1, twistLimitDegrees=45}
 {bone1Name=dyn_r_front_chain_04, bone2Name=dyn_r_back_chain_04, swing1LimitDegrees=1, swing2LimitDegrees=1, twistLimitDegrees=45}
 {bone1Name=dyn_r_front_chain_05, bone2Name=dyn_r_back_chain_05, swing1LimitDegrees=1, swing2LimitDegrees=1, twistLimitDegrees=45}
 {bone1Name=dyn_front_flap, bone2Name=dyn_front_flap_master, swing1LimitDegrees=5, swing2LimitDegrees=45, twistLimitDegrees=45}
 {bone1Name=dyn_back_Flap, bone2Name=dyn_back_Flap_master, swing1LimitDegrees=8, swing2LimitDegrees=30, twistLimitDegrees=3}
 {bone1Name=dyn_front_flap01, bone2Name=dyn_front_flap, swing1LimitDegrees=5, swing2LimitDegrees=45, twistLimitDegrees=45}
 {bone1Name=dyn_front_flap02, bone2Name=dyn_front_flap01, swing1LimitDegrees=5, swing2LimitDegrees=45, twistLimitDegrees=45}
]
