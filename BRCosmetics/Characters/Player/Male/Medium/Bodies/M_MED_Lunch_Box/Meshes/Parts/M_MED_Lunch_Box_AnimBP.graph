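# Post-process-style AnimBP for the Lunch Box medium male body part, copying the attached parent mesh pose and adding hair secondary motion.
# Two AnimDynamics setups on dyn_hair_top, both limited by LOD threshold 1, are switched by bIsInFrontEnd, with a stiffer, much lower-alpha version for the front end.
# A constant def_hairCorrBack correction follows, plus extra hair and head adjustments whose alpha the EventGraph sets to 1 at LOD 2 or higher.
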AnimBlueprint /BRCosmetics/Characters/Player/Male/Medium/Bodies/M_MED_Lunch_Box/Meshes/Parts/M_MED_Lunch_Box_AnimBP
target_skeleton=&/M_MED_Lunch_Box_Skeleton
graphs [
 AnimGraph {nodes [
   n1 = Anim Dynamics - Bone: dyn_hair_top(
    ComponentPose=Local To Component(LocalPose=Copy Pose From Mesh() {bCopyCurves=true, bUseAttachedParent=true})
    Alpha=0.6
    ExternalForce="0,0,0"
   ) {
    angularDampingOverride=0.97
    angularSpringConstant=5
    bAngularSpring=true
    bGravityOverrideInSimSpace=true
    bLinearSpring=true
    bOverrideAngularDamping=true
    lODThreshold=1
    linearDampingOverride=0.97
    linearSpringConstant=250
    numSolverIterationsPostUpdate=4
    numSolverIterationsPreUpdate=16
    simulationSpace=World
    alphaBoolBlend {blendOption=Linear}
    boundBone {boneName=dyn_hair_top}
    physicsBodyDefinitions [{
      sphereCollisionRadius=0
      boundBone {boneName=dyn_hair_top}
      boxExtents=<10 10 10>
      constraintSetup {
       angularTargetAxis=AxisZ
       twistAxis=AxisZ
       angularLimitsMax=<0 0 0>
       angularLimitsMin=<0 0 0>
       angularTarget=<0 0 0>
       linearAxesMax=<0 0.3 0>
       linearAxesMin=<0 -0.3 0>
      }
      localJointOffset=<0 0 -5>
     }]
    relativeSpaceBone {boneName=head}
   }
   n2 = Component To Local(ComponentPose=@n1)
   n3 = Anim Dynamics - Bone: dyn_hair_top(
    ComponentPose=Local To Component(LocalPose=Copy Pose From Mesh() {bCopyCurves=true, bUseAttachedParent=true})
    Alpha=0.06
    ExternalForce="0,0,0"
   ) {
    angularDampingOverride=0.97
    angularSpringConstant=15
    bAngularSpring=true
    bGravityOverrideInSimSpace=true
    bLinearSpring=true
    bOverrideAngularDamping=true
    lODThreshold=1
    linearDampingOverride=0.97
    linearSpringConstant=750
    numSolverIterationsPostUpdate=4
    numSolverIterationsPreUpdate=16
    simulationSpace=World
    alphaBoolBlend {blendOption=Linear}
    boundBone {boneName=dyn_hair_top}
    physicsBodyDefinitions [{
      sphereCollisionRadius=0
      boundBone {boneName=dyn_hair_top}
      boxExtents=<5 5 2.5>
      constraintSetup {
       angularTargetAxis=AxisZ
       twistAxis=AxisZ
       angularLimitsMax=<0 0 0>
       angularLimitsMin=<0 0 0>
       angularTarget=<0 0 0>
       linearAxesMax=<0 0.3 0>
       linearAxesMin=<0 -0.3 0>
      }
      localJointOffset=<0 0 -5>
     }]
    relativeSpaceBone {boneName=head}
   }
   n4 = Component To Local(ComponentPose=@n3)
   n5 = Get bIsInFrontEnd()
   n6 = "Transform (Modify) Bone - Bone: def_hairCorrBack"(
    ComponentPose=Local To Component(LocalPose=Blend Poses by bool(BlendPose_0=@n4, BlendPose_1=@n2, bActiveValue=@n5) {blendType=Linear})
    Alpha=1
    Translation="0,-0.2,0"
    Rotation="0,0,0"
    Scale="1,1.05,1"
   ) {
    scaleMode=BMM_Additive
    scaleSpace=BCS_BoneSpace
    translationMode=BMM_Additive
    translationSpace=BCS_BoneSpace
    alphaBoolBlend {blendOption=Linear}
    boneToModify {boneName=def_hairCorrBack}
    scale {y=1.05}
    translation {y=-0.2}
   }
   n7 = Get isLOD2OrHigh() -> (isLOD2OrHigh:double)
   n8 = "Transform (Modify) Bone - Bone: def_hairCorrBack"(ComponentPose=@n6, Alpha=@n7, Translation="0,-0.5,0", Rotation="0,0,0", Scale="1,1,1") {
    scaleMode=BMM_Additive
    scaleSpace=BCS_BoneSpace
    translationMode=BMM_Additive
    translationSpace=BCS_BoneSpace
    alphaBoolBlend {blendOption=Linear}
    boneToModify {boneName=def_hairCorrBack}
    translation {y=-0.5}
   }
   n9 = "Transform (Modify) Bone - Bone: head"(ComponentPose=@n8, Alpha=@n7, Translation="0,-0.2,0", Rotation="0,0,0", Scale="1,1,1.02") {
    scaleMode=BMM_Additive
    scaleSpace=BCS_BoneSpace
    translationMode=BMM_Additive
    translationSpace=BCS_BoneSpace
    alphaBoolBlend {blendOption=Linear}
    boneToModify {boneName=head}
    scale {z=1.02}
    translation {y=-0.2}
   }
   Output Pose(Result=Component To Local(ComponentPose=@n9)) {layerGroup=None}
  ]}
 EventGraph {nodes [
   Event BlueprintUpdateAnimation()
   n1 = Get CurrentLOD()
   > Set isLOD2OrHigh(isLOD2OrHigh:double="To Float (Boolean)"(InBool="integer >= integer"(A=@n1, B=2))) -> (Output_Get:double)
  ]}
]
variables [isLOD2OrHigh {type=number}]
